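# Spectator HUD widget that shows reboot status: card expiry countdown, waiting to be rebooted, or reboot in progress.
# Resurrection state changes select the display; loading screens and the HideExpiration tag can hide it, and a re-arming timer updates the countdown.
# The progress bar is a fill animation, seeked and speed-scaled from the reboot van's multi-interact duration and multiplier over the curve-table interact time.
# Interactor changes also update enhanced-reboot material and color.
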WidgetBlueprint /Game/Athena/HUD/Spectator/AthenaSpectatorRebootWidget
animations [
 Anim_ChipExpiresIn {
  binding_count=3
  length_seconds=0.4
  object=&self:Anim_ChipExpiresIn
  sequence {bindings [
    MessageText_2 {
     id=#1
     possessed_class=CommonRichTextBlock
     tracks [
      {
       type=MovieSceneFloatTrack
       display_name=Render Opacity
       sections [{type=MovieSceneFloatSection, end_frame=8, start_frame=0, channels [None {type=MovieSceneScriptingFloatChannel, keys [[0,0],[3,0],[8,1]]}]}]
      }
      {
       type=MovieScene2DTransformTrack
       sections [{
         type=MovieScene2DTransformSection
         end_frame=8
         start_frame=2
         channels [
          Translation.X {type=MovieSceneScriptingFloatChannel, keys [[2,50],[8,0]]}
          Translation.Y {type=MovieSceneScriptingFloatChannel, const [0,2]}
          Angle {type=MovieSceneScriptingFloatChannel}
          Scale.X {type=MovieSceneScriptingFloatChannel}
          Scale.Y {type=MovieSceneScriptingFloatChannel}
          Shear.X {type=MovieSceneScriptingFloatChannel}
          Shear.Y {type=MovieSceneScriptingFloatChannel}
         ]
        }]
      }
     ]
    }
    MessageText_1 {
     id=#2
     possessed_class=CommonRichTextBlock
     tracks [
      {
       type=MovieSceneFloatTrack
       display_name=Render Opacity
       sections [{type=MovieSceneFloatSection, end_frame=8, start_frame=0, channels [None {type=MovieSceneScriptingFloatChannel, keys [[0,0],[3,0],[8,1]]}]}]
      }
      {
       type=MovieScene2DTransformTrack
       sections [{
         type=MovieScene2DTransformSection
         end_frame=8
         start_frame=2
         channels [
          Translation.X {type=MovieSceneScriptingFloatChannel, keys [[2,-75],[8,0]]}
          Translation.Y {type=MovieSceneScriptingFloatChannel, const [0,2]}
          Angle {type=MovieSceneScriptingFloatChannel}
          Scale.X {type=MovieSceneScriptingFloatChannel}
          Scale.Y {type=MovieSceneScriptingFloatChannel}
          Shear.X {type=MovieSceneScriptingFloatChannel}
          Shear.Y {type=MovieSceneScriptingFloatChannel}
         ]
        }]
      }
     ]
    }
    TimeLeftText {
     id=#3
     possessed_class=CommonNumericTextBlock
     tracks [
      {
       type=MovieSceneFloatTrack
       display_name=Render Opacity
       sections [{type=MovieSceneFloatSection, end_frame=8, start_frame=0, channels [None {type=MovieSceneScriptingFloatChannel, keys [[0,0],[2,1],[8,1]]}]}]
      }
      {
       type=MovieScene2DTransformTrack
       sections [{
         type=MovieScene2DTransformSection
         end_frame=8
         start_frame=0
         channels [
          Translation.X {type=MovieSceneScriptingFloatChannel}
          Translation.Y {type=MovieSceneScriptingFloatChannel}
          Angle {type=MovieSceneScriptingFloatChannel}
          Scale.X {type=MovieSceneScriptingFloatChannel, keys [[0,1.3408],[2,1.3915],[5,1],[8,1]]}
          Scale.Y {type=MovieSceneScriptingFloatChannel, keys [[0,1.3408],[2,1.3915],[5,1],[8,1]]}
          Shear.X {type=MovieSceneScriptingFloatChannel}
          Shear.Y {type=MovieSceneScriptingFloatChannel}
         ]
        }]
      }
     ]
    }
   ]}
 }
 RebootProgressBarFillUp {
  binding_count=1
  length_seconds=1
  object=&self:RebootProgressBarFillUp
  sequence {bindings [Image_ResurrectionChipExpirationBar {
     id=#4
     possessed_class=Image
     tracks [{
       type=MovieSceneWidgetMaterialTrack
       sections [{
         type=MovieSceneParameterSection
         end_frame=20
         start_frame=0
         channels [Progress {type=MovieSceneScriptingFloatChannel, keys [[0,0],[20,1]]}]
        }]
      }]
    }]}
 }
 Anim_ProgressBarFull {
  binding_count=2
  length_seconds=0.45
  object=&self:Anim_ProgressBarFull
  sequence {bindings [
    Border_Progress {
     id=#5
     possessed_class=Border
     tracks [
      {
       type=MovieSceneColorTrack
       display_name=Brush Color
       sections [{
         type=MovieSceneColorSection
         end_frame=9
         start_frame=0
         channels [
          Color.R {type=MovieSceneScriptingFloatChannel}
          Color.G {type=MovieSceneScriptingFloatChannel}
          Color.B {type=MovieSceneScriptingFloatChannel}
          Color.A {type=MovieSceneScriptingFloatChannel, keys [[0,0],[3,1],[7,0.2],[9,0]]}
         ]
        }]
      }
      {
       type=MovieScene2DTransformTrack
       sections [{
         type=MovieScene2DTransformSection
         end_frame=9
         start_frame=0
         channels [
          Translation.X {type=MovieSceneScriptingFloatChannel}
          Translation.Y {type=MovieSceneScriptingFloatChannel}
          Angle {type=MovieSceneScriptingFloatChannel}
          Scale.X {type=MovieSceneScriptingFloatChannel, const [1,4]}
          Scale.Y {type=MovieSceneScriptingFloatChannel, keys [[0,1],[5,1.7707],[7,1.7161],[9,1.7161]]}
          Shear.X {type=MovieSceneScriptingFloatChannel}
          Shear.Y {type=MovieSceneScriptingFloatChannel}
         ]
        }]
      }
     ]
    }
    Image_ResurrectionChipExpirationBar {
     id=#6
     possessed_class=Image
     tracks [{
       type=MovieSceneWidgetMaterialTrack
       sections [{
         type=MovieSceneParameterSection
         end_frame=9
         start_frame=0
         channels [
          OverrideWithColorB {type=MovieSceneScriptingFloatChannel, keys [[0,0],[7,1]]}
          Position {type=MovieSceneScriptingFloatChannel, keys [[0,0],[5,1]]}
          PulseColorStrength {type=MovieSceneScriptingFloatChannel, keys [[0,0],[3,0],[9,1]]}
         ]
        }]
      }]
    }
   ]}
 }
]
graphs [
 EventGraph {nodes [
   DisplayChipAvailableMessage() {is=title K2Node_CustomEvent}
   > SetText(
    self=Get MessageText_1() -> (MessageText_1:Common Rich Text Block Object Reference)
    InText=TextToUpper!(InText:"Text (by ref)"=Get GrabCard() -> (GrabCard:Text)) -> (ReturnValue:Text)
   )
   > PlayAnimation(InAnimation=Get Anim_ChipExpiresIn() -> (Anim_ChipExpiresIn:Widget Animation Object Reference))
   > SetVisibility(
    self=[
     Get TimeLeftText() -> (TimeLeftText:Common Numeric Text Block Object Reference)
     Get MessageText_1()
     Get MessageText_2() -> (MessageText_2:Common Rich Text Block Object Reference)
    ]
    InVisibility=HitTestInvisible
   )
   > HandleCardAvailableTimer()
   > SetVisibility(InVisibility=HitTestInvisible)
   n1 = HandleOnLoadScreenChanged!() -> (OutputDelegate, then, PlayerController:Fort Player Controller Athena Object Reference, bEnableLoadScreen:bool=~, HUDReasonText:Text) {is=title K2Node_CustomEvent}
   > n2 = Set LoadingScreenEnabled(LoadingScreenEnabled:bool=@n1.bEnableLoadScreen) -> (Output_Get:bool)
   > n3 = Branch(Condition=@n2)
   > SetVisibility(InVisibility=Collapsed)
   HandleOnResurrectionStateChanged(execute=@n3.else)
   PlayRebootProgress() {is=title K2Node_CustomEvent}
   n4 = GetMultiInteractComponent!(self) -> (ReturnValue:Fort Actor Component Multi Interact Object Reference)
   > n5 = Branch(Condition="Equal (Object)"(A=Get CachedVanMultiInteractComponent() -> (CachedVanMultiInteractComponent:Fort Actor Component Multi Interact Object Reference), B=@n4))
   > n6 = Is Valid(InputObject=Get CachedVanMultiInteractComponent())
   n7 = Get CachedVanMultiInteractComponent()
   PlayRebootAnimation!(
    execute=[@n13.then, @n6.Is Valid]
    self
    PlayMultiplier:double=GetMultiplier!(self:Fort Actor Component Multi Interact Object Reference=@n7) -> (ReturnValue:float)
    StartDuration:double=GetInteractionDuration!(self:Fort Actor Component Multi Interact Object Reference=@n7) -> (ReturnValue:float)
   ) -> (then)
   n8 = Get CachedVanMultiInteractComponent()
   n9 = Is Valid(exec=@n5.else, InputObject=@n8)
   n10 = HandleRebootActorChanged!() -> (OutputDelegate, then, InteractingActorCount:int, MultiInteractMultiplier:float=~, InteractingDuration:float=~) {is=title K2Node_CustomEvent}
   n11 = Unbind Event from Actor Count Changed Delegate(execute=@n9.Is Valid, self:Fort Actor Component Multi Interact Object Reference=@n8, Delegate=@n10) {is=title K2Node_RemoveDelegate}
   n12 = Set CachedVanMultiInteractComponent(execute=[@n11.then, @n9.Is Not Valid], CachedVanMultiInteractComponent:Fort Actor Component Multi Interact Object Reference=@n4) -> (Output_Get:Fort Actor Component Multi Interact Object Reference)
   > Is Valid(InputObject=@n12)
   > n13 = Bind Event to Actor Count Changed Delegate(self:Fort Actor Component Multi Interact Object Reference=@n12, Delegate=@n10) {is=title K2Node_AddDelegate}
   n14 = OnFillComplete() {is=title K2Node_CustomEvent}
   > PlayAnimation(InAnimation=Get Anim_ProgressBarFull() -> (Anim_ProgressBarFull:Widget Animation Object Reference))
   ResetProgressBar() {is=title K2Node_CustomEvent}
   n15 = Get Anim_ProgressBarFull()
   > PlayAnimation(InAnimation=@n15, StartAtTime=GetEndTime(self=@n15), PlayMode=Reverse)
   HandleCardAvailableTimer() {is=title K2Node_CustomEvent}
   n16 = GetRemainingResurrectionChipExpirationTime!(self:Fort Player State Athena Object Reference=Get PlayerState() -> (PlayerState:Fort Player State Athena Object Reference)) -> (ReturnValue:float)
   > SetCurrentValue!(
    execute
    self:Common Numeric Text Block Object Reference=Get TimeLeftText()
    NewValue:float=Select(
     Option 0:double=@n16
     Option 1:double=0
     Index="float <= float"!(A:double=@n16, B:double=0) -> (ReturnValue:bool) {is=title K2Node_CallFunction}
    ) -> (ReturnValue:double)
   ) -> (then)
   n17 = Get TimeLeftText()
   > Branch(Condition="float > float"(A=Get CurrentNumericValue!(self:Common Numeric Text Block Object Reference=@n17) -> (CurrentNumericValue:float), B=0) {is=title K2Node_CallFunction})
   > n18 = Set Timer by Event(Delegate=Create Event(), Time=0.25, bLooping=false)
   > Set CardTimerHandle(CardTimerHandle:Timer Handle Structure=@n18) -> (Output_Get:Timer Handle Structure)
   n19 = GetOwningPlayer()
   n20 = Cast To FortPlayerControllerAthena(Object=@n19)
   n21 = Create Event()
   > Bind Event to In Game Load Screen Changed(self:Fort Player Controller Athena Object Reference=@n20, Delegate=@n21) {is=title K2Node_AddDelegate}
   Event Construct()
   > BindToAnimationFinished(Animation=Get RebootProgressBarFillUp() -> (RebootProgressBarFillUp:Widget Animation Object Reference), Delegate=@n14)
   > n22 = GetDynamicMaterial(self=Get Image_BlueCircle() -> (Image_BlueCircle:Image Object Reference))
   > n23 = Set M_ColoredCircle(M_ColoredCircle:Material Instance Dynamic Object Reference=@n22) -> (Output_Get:Material Instance Dynamic Object Reference)
   > SetVectorParameterValue(self=@n23, ParameterName=Color_1, Value=(R=0,G=0.4452,B=1,A=1))
   > SetVisibility(InVisibility=Collapsed)
   n24 = GetOwningPlayer()
   > n25 = GetComponentByClass!(execute, self:Actor Object Reference=@n24, ComponentClass:Actor Component Class Reference=/Script/FortniteGame.AthenaResurrectionComponent) -> (then, ReturnValue:Athena Resurrection Component Object Reference)
   > Set CachedResurrectionComponent(CachedResurrectionComponent:Athena Resurrection Component Object Reference=@n25) -> (Output_Get:Athena Resurrection Component Object Reference)
   > n26 = WaitGameplayTagCountChangedOnController!(
    execute:Exec
    Owner:Object Reference=GetAbilitySystemComponent!(Actor:Actor Object Reference=@n24) -> (ReturnValue:Ability System Component Object Reference)
    TargetPlayerController:Controller Object Reference=@n24
    RelevantTag:Gameplay Tag Structure=(TagName="Athena.RebootCard.HideExpiration")
    Params:"Fort Async Action Listen For Tag Count Change Params Structure (by ref)"=Get Params() -> (Params:Fort Async Action Listen For Tag Count Change Params Structure)
   ) -> (then:Exec, OnTagCountChanged:Exec, NewTagCount:int, TagTriggered:Gameplay Tag Structure) {is=title K2Node_AsyncAction}
   n27 = IsResurrectionChipAvailable!(self:Fort Player State Athena Object Reference=GetPlayerStateFromActor!(Actor:Actor Object Reference=GetOwningPlayer(), PlayerStateType:Player State Class Reference=/Script/FortniteGame.FortPlayerStateAthena) -> (ReturnValue:Fort Player State Athena Object Reference)) -> (ReturnValue:bool)
   n28 = Branch(execute=@n26.OnTagCountChanged, Condition=OR Boolean(A="integer > integer"(A=@n26), B=NOT Boolean(A=@n27)))
   > SetVisibility(InVisibility=Hidden)
   > SetVisibility(self=Get BorderProgressHolder() -> (BorderProgressHolder:Vertical Box Object Reference), InVisibility=Hidden)
   > Clear and Invalidate Timer by Handle(Handle=Get DisplayMessageHearbeatTimer() -> (DisplayMessageHearbeatTimer:Timer Handle Structure))
   > Clear and Invalidate Timer by Handle(Handle=Get CardTimerHandle() -> (CardTimerHandle:Timer Handle Structure))
   SetVisibility(execute=@n28.else, InVisibility=Visible)
   > DisplayChipAvailableMessage() {is=title K2Node_CallFunction}
   Event Destruct()
   > n29 = Cast To FortPlayerControllerAthena(Object=@n19)
   > Unbind Event from In Game Load Screen Changed(self:Fort Player Controller Athena Object Reference=@n29, Delegate=@n21) {is=title K2Node_RemoveDelegate}
   Event OnResurrectionStateChanged() {is=title K2Node_Event}
   > HandleOnResurrectionStateChanged()
   Event PreConstruct()
   > SetMobileStyles()
   n30 = Get M_ColoredCircle() -> (M_ColoredCircle:Material Instance Dynamic Object Reference)
   SetVectorParameterValue(execute=@n35.then, self=@n30, ParameterName=Color_1, Value=(R=1,G=0.3005,B=0,A=1))
   SetVectorParameterValue(execute=@n36.then, self=@n30, ParameterName=Color_1, Value=(R=0,G=0.4452,B=1,A=1))
   OnVisi() {is=title K2Node_CustomEvent}
   n31 = GetValueAtLevel!(Input:"Scalable Float Structure (by ref)"=Get Row_InteractTime() -> (Row_InteractTime:Scalable Float Structure), Level:float=1) -> (ReturnValue:float)
   n32 = GetDynamicMaterial(execute=@n38.then, self=Get Image_ResurrectionChipExpirationBar() -> (Image_ResurrectionChipExpirationBar:Image Object Reference))
   n33 = Get CachedResurrectionComponent() -> (CachedResurrectionComponent:Athena Resurrection Component Object Reference)
   > n34 = Branch(Condition="Not Equal (Boolean)"(
     A=Get IsEnhancedReboot() -> (IsEnhancedReboot:bool)
     B=GetIsEnhancedResurrectionDesiredLocal!(self:Athena Resurrection Component Object Reference=@n33) -> (ReturnValue:bool)
    ) -> (ReturnValue=~))
   > n35 = SetScalarParameterValue(self=@n32, ParameterName=EnhancedSwitch, Value=1)
   n36 = SetScalarParameterValue(execute=@n34.else, self=@n32, ParameterName=EnhancedSwitch, Value=0)
   n37 = Get RebootProgressBarFillUp()
   SetAnimationCurrentTime!(
    execute=@n39.Is Valid
    self:User Widget Object Reference
    InAnimation:Widget Animation Object Reference=@n37
    InTime:float=float / float(A=@n10.InteractingDuration, B=@n31)
   ) -> (then)
   > n38 = SetPlaybackSpeed!(
    execute
    self:User Widget Object Reference
    InAnimation:Widget Animation Object Reference=@n37
    PlaybackSpeed:float=float / float(A=@n10.MultiInteractMultiplier, B=@n31)
   ) -> (then)
   Branch(execute=@n10.then, Condition="integer > integer"(A=@n10.InteractingActorCount))
   > n39 = Is Valid(InputObject=@n33)
  ]}
 GetMultiInteractComponent {nodes [
   GetMultiInteractComponent() {is=title K2Node_FunctionEntry}
   n1 = GetCurrentInteractingRebootVan!(self:Fort Spectator Reboot Widget Object Reference) -> (ReturnValue:Spawn Machine Interface Interface)
   > n2 = Cast To Actor!(execute, Object:Spawn Machine Interface Interface=@n1) -> (then, CastFailed, AsActor:Actor Object Reference) {is=title K2Node_DynamicCast}
   > Is Valid(InputObject=@n2)
   > Return Node!(
    execute
    ReturnValue:Fort Actor Component Multi Interact Object Reference=GetComponentByClass!(self=@n2, ComponentClass=/Script/FortniteGame.FortActorComponent_MultiInteract) -> (ReturnValue:Fort Actor Component Multi Interact Object Reference)
   )
   n3 = Cast To ActorComponent!(execute=@n2.CastFailed, Object:Spawn Machine Interface Interface=@n1) -> (then, CastFailed, AsActor Component:Actor Component Object Reference) {is=title K2Node_DynamicCast}
   n4 = GetOwner(self=@n3)
   > Is Valid(InputObject=@n4)
   > Return Node(ReturnValue=GetComponentByClass(self=@n4, ComponentClass=&FortActorComponent_MultiInteract))
  ]}
 HandleOnResurrectionStateChanged {nodes [
   HandleOnResurrectionStateChanged() {is=title K2Node_FunctionEntry}
   > SetVisibility(InVisibility=Collapsed)
   > SetVisibility(self=Get BorderProgressHolder(), InVisibility=Hidden)
   > Clear and Invalidate Timer by Handle(Handle=Get DisplayMessageHearbeatTimer())
   > Clear and Invalidate Timer by Handle(Handle=Get CardTimerHandle())
   > ResetProgressBar() {is=title K2Node_CallFunction}
   > Branch(Condition=NOT Boolean(A=Get LoadingScreenEnabled() -> (LoadingScreenEnabled:bool)))
   > n1 = Get!(execute:Exec) -> (then:Exec, else:Exec, PlayerState:Fort Player State Athena Object Reference)
   > n2 = Branch(Condition=IsResurrectionChipAvailable(self=@n1))
   > n3 = Set Timer by Event(Delegate=Create Event(), Time=Get HeartbeatDisplayDelayTime() -> (HeartbeatDisplayDelayTime:double), bLooping=false)
   > Set DisplayMessageHearbeatTimer(DisplayMessageHearbeatTimer:Timer Handle Structure=@n3) -> (Output_Get:Timer Handle Structure)
   > StopAnimation(InAnimation=Get RebootProgressBarFillUp())
   n4 = IsResurrectingNow!(self:Fort Player State Athena Object Reference=@n1) -> (ReturnValue:bool)
   n5 = Branch(
    execute=@n2.else
    Condition=AND Boolean(A=CanBeResurrected!(self:Fort Player State Athena Object Reference=@n1) -> (ReturnValue:bool), B=NOT Boolean(A=@n4))
   )
   > SetText(self=Get MessageText_1(), InText=TextToUpper(InText=Get AwaitingReboot() -> (AwaitingReboot:Text)))
   > SetVisibility(self=Get MessageText_1(), InVisibility=HitTestInvisible)
   > SetVisibility(self=[Get MessageText_2(), Get TimeLeftText()], InVisibility=Collapsed)
   > SetVisibility(InVisibility=HitTestInvisible)
   > StopAnimation(InAnimation=Get RebootProgressBarFillUp())
   n6 = Is Valid(
    exec=@n5.else
    InputObject="To Object (Interface)"!(Interface:"Interface Interface (by ref)"=GetCurrentInteractingRebootVan()) -> (ReturnValue:Object Reference)
   )
   n7 = IsResurrectingNow(self=@n1)
   Branch(execute=@n6.Is Not Valid, Condition=@n7)
   > SetText(self=Get MessageText_1(), InText=TextToUpper(InText=Get RebootMomentarily() -> (RebootMomentarily:Text)))
   > SetVisibility(self=Get MessageText_1(), InVisibility=HitTestInvisible)
   > SetVisibility(self=[Get MessageText_2(), Get BorderProgressHolder(), Get TimeLeftText()], InVisibility=Collapsed)
   > SetVisibility(InVisibility=HitTestInvisible)
   n8 = GetSpawnMachineState!(SpawnMachine:"Spawn Machine Interface Interface (by ref)"=GetCurrentInteractingRebootVan()) -> (ReturnValue:ESpawnMachineState Enum=Default)
   n9 = "Equal (Enum)"(A:ESpawnMachineState Enum=@n8, B:ESpawnMachineState Enum=BeingUsed)
   n10 = "Equal (Enum)"(A=@n8, B=WaitingForUse)
   Branch(execute=@n6.Is Valid, Condition=AND Boolean(A=OR Boolean(A=@n9, B=@n10), B=@n7))
   > SetText(self=Get MessageText_1(), InText=TextToUpper(InText=Get RebootMomentarily()))
   > SetVisibility(self=[Get BorderProgressHolder(), Get MessageText_1()], InVisibility=HitTestInvisible)
   > SetVisibility(self=[Get MessageText_2(), Get TimeLeftText()], InVisibility=Collapsed)
   > SetVisibility(InVisibility=HitTestInvisible)
   > PlayRebootProgress() {is=title K2Node_CallFunction}
  ]}
 PlayRebootAnimation {nodes [
   n1 = PlayRebootAnimation!() -> (then, PlayMultiplier:double=~, StartDuration:double=~) {is=title K2Node_FunctionEntry}
   > Is Valid(InputObject=Get CachedResurrectionComponent())
   n2 = GetValueAtLevel(Input=Get Row_InteractTime(), Level=1)
   > PlayAnimation(InAnimation=Get RebootProgressBarFillUp(), StartAtTime=float / float(A=@n1.StartDuration, B=@n2), PlaybackSpeed=float / float(A=@n1, B=@n2))
  ]}
 SetMobileStyles {nodes [
   SetMobileStyles() {is=title K2Node_FunctionEntry}
   n1 = GetOwningPlayer()
   > Branch(Condition=AND Boolean(A=IsMobileGame(), B=IsUsingTouch(PlayerController=@n1)))
   > Set!(
    execute
    Padding_Left:float=7
    Padding_Top:float=10
    Padding_Right:float=3
    Padding_Bottom:float=0
    self:Horizontal Box Slot Object Reference=SlotAsHorizontalBoxSlot!(Widget:Widget Object Reference=Get TimeLeftText()) -> (ReturnValue:Horizontal Box Slot Object Reference)
   ) -> (then, Output_Get:Margin Structure)
  ]}
]
variables [
 AwaitingReboot {type=string, default=waiting to be rebooted}
 CachedResurrectionComponent {type=AthenaResurrectionComponent, default=null}
 CachedVanMultiInteractComponent {type=FortActorComponent_MultiInteract, default=null}
 CardTimerHandle {type=TimerHandle}
 DisplayMessageHearbeatTimer {type=TimerHandle}
 GrabCard {type=string, default="Reboot card expires in "}
 HeartbeatDisplayDelayTime {type=number, default=3}
 IsEnhancedReboot {type=boolean}
 LoadingScreenEnabled {type=boolean}
 M_ColoredCircle {type=MaterialInstanceDynamic}
 Params {type=FortAsyncAction_ListenForTagCountChangeParams, default {bBroadcastAllTagChanges=false, bBroadcastInitialValues=false}}
 RebootMomentarily {type=string, default=processing reboot}
 Row_InteractTime {
  type=ScalableFloat
  default {registryType=None, value=1, curve {curveTable=/Game/Athena/Balance/DataTables/AthenaGameData, rowName=Default.SCMachine.InteractTime}}
 }
 Tint Color {type=SlateColor, default {colorUseRule=UseColor_Specified, specifiedColor=<rgba 0.1304 0.4201 0 1>}}
]
widgets {
 info {inheritedWidgetCount=1, parentClass=/Script/FortniteUI.FortSpectatorRebootWidget, rootWidgetClass=&VerticalBox, widgetCount=15}
 widgets [
  VerticalBox_0 {widgetClassPath=&VerticalBox, widget=VerticalBox_0, properties {bOverride_Cursor=true}}
  BorderProgressHolder {
   widgetClassPath=&VerticalBox
   bIsVariable=true
   parent=VerticalBox_0
   widget=BorderProgressHolder
   properties {visibility=HitTestInvisible}
   slot {class=VerticalBoxSlot}
  }
  SizeBox_0 {
   widgetClassPath=&SizeBox
   parent=BorderProgressHolder
   widget=SizeBox_0
   properties {bOverride_WidthOverride=true, widthOverride=520}
   slot {class=VerticalBoxSlot, properties {padding=<ltrb 0 -8 0 -8>}}
  }
  Overlay_1 {widgetClassPath=&Overlay, parent=SizeBox_0, widget=Overlay_1, slot {class=SizeBoxSlot}}
  White_Background {
   widgetClassPath=&Image
   bIsVariable=true
   parent=Overlay_1
   widget=White_Background
   properties {
    visibility=HitTestInvisible
    brush {
     bIsDynamicallyLoaded=false
     drawAs=Image
     imageType=NoImage
     mirroring=NoMirror
     resourceName=None
     resourceObject=null
     tiling=NoTile
     imageSize=<15 20>
     margin=<ltrb 0 0 0 0>
     outlineSettings {
      bUseBrushTransparency=false
      roundingType=HalfHeightRadius
      width=0
      color {colorUseRule=UseColor_Specified, specifiedColor=<rgba 0 0 0 0>}
      cornerRadii=<0 0 0 0>
     }
     tintColor {colorUseRule=UseColor_Specified, specifiedColor=<rgba 1 1 1 1>}
     uVRegion {bIsValid=false, max=<0 0>, min=<0 0>}
    }
    renderTransform {angle=0, scale=<1 1>, shear=<-12 0>, translation=<0 0>}
   }
   slot {class=OverlaySlot, properties {horizontalAlignment=HAlign_Fill, verticalAlignment=VAlign_Center, padding=<ltrb 7 5 7 5>}}
  }
  Border_Progress {
   widgetClassPath=&Border
   bIsVariable=true
   parent=Overlay_1
   widget=Border_Progress
   properties {
    visibility=HitTestInvisible
    background {
     bIsDynamicallyLoaded=false
     drawAs=Box
     imageType=NoImage
     mirroring=NoMirror
     resourceName=None
     resourceObject=/Game/UI/Frontend/Rewards/Textures/shadow_blur
     tiling=NoTile
     imageSize=<32 32>
     margin=<ltrb 0.3 0 0.3 0>
     outlineSettings {
      bUseBrushTransparency=false
      roundingType=HalfHeightRadius
      width=0
      color {colorUseRule=UseColor_Specified, specifiedColor=<rgba 0 0 0 0>}
      cornerRadii=<0 0 0 0>
     }
     tintColor {colorUseRule=UseColor_Specified, specifiedColor=<rgba 1 1 1 1>}
     uVRegion {bIsValid=false, max=<0 0>, min=<0 0>}
    }
    brushColor=<rgba 0 1 0.9405 0>
    padding=<ltrb 10 10 50 10>
   }
   slot {class=OverlaySlot, properties {horizontalAlignment=HAlign_Fill, verticalAlignment=VAlign_Center}}
  }
  Image_ResurrectionChipExpirationBar {
   widgetClassPath=&Image
   bIsVariable=true
   parent=Border_Progress
   widget=Image_ResurrectionChipExpirationBar
   properties {
    visibility=SelfHitTestInvisible
    brush {
     bIsDynamicallyLoaded=false
     drawAs=Image
     imageType=NoImage
     mirroring=NoMirror
     resourceName=None
     resourceObject=/Game/Athena/HUD/Team/MI_Player_RebootBar
     tiling=NoTile
     imageSize=<24 12>
     margin=<ltrb 0 0 0 0>
     outlineSettings {
      bUseBrushTransparency=false
      roundingType=HalfHeightRadius
      width=0
      color {colorUseRule=UseColor_Specified, specifiedColor=<rgba 0 0 0 0>}
      cornerRadii=<0 0 0 0>
     }
     tintColor {colorUseRule=UseColor_Specified, specifiedColor=<rgba 1 1 1 1>}
     uVRegion {bIsValid=false, max=<0 0>, min=<0 0>}
    }
    renderTransform {angle=0, scale=<1 1>, shear=<-12 0>, translation=<0 0>}
   }
   slot {class=BorderSlot, properties {padding=<ltrb 10 10 50 10>}}
  }
  Image_WhiteCircle {
   widgetClassPath=&Image
   bIsVariable=true
   parent=Overlay_1
   widget=Image_WhiteCircle
   properties {
    visibility=HitTestInvisible
    brush {
     bIsDynamicallyLoaded=false
     drawAs=Image
     imageType=FullColor
     mirroring=NoMirror
     resourceName=None
     resourceObject=/FortUILibrary/Foundation/Art_ToolKit/Tools/M_UI_Circle_White
     tiling=NoTile
     imageSize=<52 52>
     margin=<ltrb 0 0 0 0>
     outlineSettings {
      bUseBrushTransparency=false
      roundingType=HalfHeightRadius
      width=0
      color {colorUseRule=UseColor_Specified, specifiedColor=<rgba 0 0 0 0>}
      cornerRadii=<0 0 0 0>
     }
     tintColor {colorUseRule=UseColor_Specified, specifiedColor=<rgba 1 1 1 1>}
     uVRegion {bIsValid=false, max=<0 0>, min=<0 0>}
    }
   }
   slot {class=OverlaySlot, properties {horizontalAlignment=HAlign_Right}}
  }
  Image_BlueCircle {
   widgetClassPath=&Image
   bIsVariable=true
   parent=Overlay_1
   widget=Image_BlueCircle
   properties {
    visibility=HitTestInvisible
    brush {
     bIsDynamicallyLoaded=false
     drawAs=Image
     imageType=FullColor
     mirroring=NoMirror
     resourceName=None
     resourceObject=&M_UI_Circle_White
     tiling=NoTile
     imageSize=<44 44>
     margin=<ltrb 0 0 0 0>
     outlineSettings {
      bUseBrushTransparency=false
      roundingType=HalfHeightRadius
      width=0
      color {colorUseRule=UseColor_Specified, specifiedColor=<rgba 0 0 0 0>}
      cornerRadii=<0 0 0 0>
     }
     tintColor {colorUseRule=UseColor_Specified, specifiedColor=<rgba 0 0.4452 1 1>}
     uVRegion {bIsValid=false, max=<0 0>, min=<0 0>}
    }
   }
   slot {class=OverlaySlot, properties {horizontalAlignment=HAlign_Right, padding=<ltrb 0 4 4 0>}}
  }
  Image_RebootIcon {
   widgetClassPath=&Image
   bIsVariable=true
   parent=Overlay_1
   widget=Image_RebootIcon
   properties {
    visibility=HitTestInvisible
    brush {
     bIsDynamicallyLoaded=false
     drawAs=Image
     imageType=FullColor
     mirroring=NoMirror
     resourceName=None
     resourceObject=/Game/Athena/HUD/Materials/MI_UI_RebootIcon
     tiling=NoTile
     imageSize=<42 42>
     margin=<ltrb 0 0 0 0>
     outlineSettings {
      bUseBrushTransparency=false
      roundingType=HalfHeightRadius
      width=0
      color {colorUseRule=UseColor_Specified, specifiedColor=<rgba 0 0 0 0>}
      cornerRadii=<0 0 0 0>
     }
     tintColor {colorUseRule=UseColor_Specified, specifiedColor=<rgba 1 1 1 1>}
     uVRegion {bIsValid=false, max=<0 0>, min=<0 0>}
    }
   }
   slot {class=OverlaySlot, properties {horizontalAlignment=HAlign_Right, verticalAlignment=VAlign_Center, padding=<ltrb 0 0 5 0>}}
  }
  HorizontalBox_0 {
   widgetClassPath=&HorizontalBox
   bIsVariable=true
   parent=VerticalBox_0
   widget=HorizontalBox_0
   slot {class=VerticalBoxSlot, properties {horizontalAlignment=HAlign_Center}}
  }
  MessageText_1 {
   widgetClassPath=&CommonRichTextBlock
   bIsVariable=true
   parent=HorizontalBox_0
   widget=MessageText_1
   properties {
    bOverrideDefaultStyle=true
    mobileTextBlockScale=1.3
    text=REBOOT CARD EXPIRES IN
    defaultTextStyleOverride {
     overflowPolicy=Clip
     transformPolicy=None
     colorAndOpacity {colorUseRule=UseColor_Specified, specifiedColor=<rgba 1 1 1 1>}
     font {
      bForceMonospaced=false
      bMaterialIsStencil=false
      fontMaterial=null
      fontObject=&Burbank
      letterSpacing=0
      monospacedWidth=1
      size=18
      skewAmount=0
      typefaceFontName=BurbankRegular-Black
      outlineSettings {
       bApplyOutlineToDropShadows=false
       bMiteredCorners=false
       bSeparateFillAlpha=true
       outlineMaterial=null
       outlineSize=0
       outlineColor=<rgba 0 0 0 0>
      }
     }
     highlightColor {colorUseRule=UseColor_Specified, specifiedColor=<rgba 0 0 0 1>}
     highlightShape {
      bIsDynamicallyLoaded=false
      drawAs=Image
      imageType=NoImage
      mirroring=NoMirror
      resourceName=None
      resourceObject=null
      tiling=NoTile
      imageSize=<32 32>
      margin=<ltrb 0 0 0 0>
      outlineSettings {
       bUseBrushTransparency=false
       roundingType=HalfHeightRadius
       width=0
       color {colorUseRule=UseColor_Specified, specifiedColor=<rgba 0 0 0 0>}
       cornerRadii=<0 0 0 0>
      }
      tintColor {colorUseRule=UseColor_Specified, specifiedColor=<rgba 1 1 1 1>}
      uVRegion {bIsValid=false, max=<0 0>, min=<0 0>}
     }
     selectedBackgroundColor {colorUseRule=UseColor_Foreground, specifiedColor=<rgba 1 0 1 1>}
     shadowColorAndOpacity=<rgba 0 0 0 0.5>
     shadowOffset=<2 3>
     strikeBrush {
      bIsDynamicallyLoaded=false
      drawAs=Image
      imageType=NoImage
      mirroring=NoMirror
      resourceName=None
      resourceObject=null
      tiling=NoTile
      imageSize=<32 32>
      margin=<ltrb 0 0 0 0>
      outlineSettings {
       bUseBrushTransparency=false
       roundingType=HalfHeightRadius
       width=0
       color {colorUseRule=UseColor_Specified, specifiedColor=<rgba 0 0 0 0>}
       cornerRadii=<0 0 0 0>
      }
      tintColor {colorUseRule=UseColor_Specified, specifiedColor=<rgba 1 1 1 1>}
      uVRegion {bIsValid=false, max=<0 0>, min=<0 0>}
     }
     underlineBrush {
      bIsDynamicallyLoaded=false
      drawAs=Image
      imageType=NoImage
      mirroring=NoMirror
      resourceName=None
      resourceObject=null
      tiling=NoTile
      imageSize=<32 32>
      margin=<ltrb 0 0 0 0>
      outlineSettings {
       bUseBrushTransparency=false
       roundingType=HalfHeightRadius
       width=0
       color {colorUseRule=UseColor_Specified, specifiedColor=<rgba 0 0 0 0>}
       cornerRadii=<0 0 0 0>
      }
      tintColor {colorUseRule=UseColor_Specified, specifiedColor=<rgba 1 1 1 1>}
      uVRegion {bIsValid=false, max=<0 0>, min=<0 0>}
     }
    }
   }
   slot {class=HorizontalBoxSlot, properties {horizontalAlignment=HAlign_Left, verticalAlignment=VAlign_Center, padding=<ltrb 5 10 0 0>}}
  }
  TimeLeftText {
   widgetClassPath=/Script/CommonUI.CommonNumericTextBlock
   bIsVariable=true
   parent=HorizontalBox_0
   widget=TimeLeftText
   properties {
    currentNumericValue=30
    text=30
    colorAndOpacity {colorUseRule=UseColor_Specified, specifiedColor=<rgba 0 0.4452 1 1>}
    font {
     bForceMonospaced=false
     bMaterialIsStencil=false
     fontMaterial=null
     fontObject=&Burbank
     letterSpacing=0
     monospacedWidth=1
     size=36
     skewAmount=0
     typefaceFontName=Burbank-Black
     outlineSettings {
      bApplyOutlineToDropShadows=false
      bMiteredCorners=false
      bSeparateFillAlpha=false
      outlineMaterial=null
      outlineSize=0
      outlineColor=<rgba 0 0 0 1>
     }
    }
    formattingSpecification {
     alwaysSign=false
     maximumFractionalDigits=0
     maximumIntegralDigits=324
     minimumFractionalDigits=0
     minimumIntegralDigits=1
     roundingMode=HalfToZero
     useGrouping=true
    }
    shadowColorAndOpacity=<rgba 0 0 0 0.666>
    shadowOffset=<3 4>
   }
   slot {class=HorizontalBoxSlot, properties {padding=<ltrb 3 0 3 0>}}
  }
  MessageText_2 {
   widgetClassPath=&CommonRichTextBlock
   bIsVariable=true
   parent=HorizontalBox_0
   widget=MessageText_2
   properties {
    bOverrideDefaultStyle=true
    mobileTextBlockScale=1.3
    text=S
    defaultTextStyleOverride {
     overflowPolicy=Clip
     transformPolicy=None
     colorAndOpacity {colorUseRule=UseColor_Specified, specifiedColor=<rgba 1 1 1 1>}
     font {
      bForceMonospaced=false
      bMaterialIsStencil=false
      fontMaterial=null
      fontObject=&Burbank
      letterSpacing=0
      monospacedWidth=1
      size=18
      skewAmount=0
      typefaceFontName=BurbankRegular-Black
      outlineSettings {
       bApplyOutlineToDropShadows=false
       bMiteredCorners=false
       bSeparateFillAlpha=false
       outlineMaterial=null
       outlineSize=0
       outlineColor=<rgba 0 0 0 1>
      }
     }
     highlightColor {colorUseRule=UseColor_Specified, specifiedColor=<rgba 0 0 0 1>}
     highlightShape {
      bIsDynamicallyLoaded=false
      drawAs=Image
      imageType=NoImage
      mirroring=NoMirror
      resourceName=None
      resourceObject=null
      tiling=NoTile
      imageSize=<32 32>
      margin=<ltrb 0 0 0 0>
      outlineSettings {
       bUseBrushTransparency=false
       roundingType=HalfHeightRadius
       width=0
       color {colorUseRule=UseColor_Specified, specifiedColor=<rgba 0 0 0 0>}
       cornerRadii=<0 0 0 0>
      }
      tintColor {colorUseRule=UseColor_Specified, specifiedColor=<rgba 1 1 1 1>}
      uVRegion {bIsValid=false, max=<0 0>, min=<0 0>}
     }
     selectedBackgroundColor {colorUseRule=UseColor_Foreground, specifiedColor=<rgba 1 0 1 1>}
     shadowColorAndOpacity=<rgba 0 0 0 0.666>
     shadowOffset=<2 3>
     strikeBrush {
      bIsDynamicallyLoaded=false
      drawAs=Image
      imageType=NoImage
      mirroring=NoMirror
      resourceName=None
      resourceObject=null
      tiling=NoTile
      imageSize=<32 32>
      margin=<ltrb 0 0 0 0>
      outlineSettings {
       bUseBrushTransparency=false
       roundingType=HalfHeightRadius
       width=0
       color {colorUseRule=UseColor_Specified, specifiedColor=<rgba 0 0 0 0>}
       cornerRadii=<0 0 0 0>
      }
      tintColor {colorUseRule=UseColor_Specified, specifiedColor=<rgba 1 1 1 1>}
      uVRegion {bIsValid=false, max=<0 0>, min=<0 0>}
     }
     underlineBrush {
      bIsDynamicallyLoaded=false
      drawAs=Image
      imageType=NoImage
      mirroring=NoMirror
      resourceName=None
      resourceObject=null
      tiling=NoTile
      imageSize=<32 32>
      margin=<ltrb 0 0 0 0>
      outlineSettings {
       bUseBrushTransparency=false
       roundingType=HalfHeightRadius
       width=0
       color {colorUseRule=UseColor_Specified, specifiedColor=<rgba 0 0 0 0>}
       cornerRadii=<0 0 0 0>
      }
      tintColor {colorUseRule=UseColor_Specified, specifiedColor=<rgba 1 1 1 1>}
      uVRegion {bIsValid=false, max=<0 0>, min=<0 0>}
     }
    }
   }
   slot {class=HorizontalBoxSlot, properties {horizontalAlignment=HAlign_Left, verticalAlignment=VAlign_Center, padding=<ltrb 0 10 0 0>}}
  }
  TutorialHighlight {widgetClassPath=&TutorialHighlight}
 ]
}
