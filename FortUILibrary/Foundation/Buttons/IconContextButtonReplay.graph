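# Icon context button for replay UI, derived from WBP_UIKit_Button_Generic, that switches button style by input method and tints itself on interaction.
# Construct copies the current style into both controller and mouse/keyboard slots, which setters can override.
# UpdateStyle picks a slot when the input method or triggered action changes.
# OnHover animates white to orange, reversing on unhover or release, while click and press play OnSelected to hold orange.
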WidgetBlueprint /FortUILibrary/Foundation/Buttons/IconContextButtonReplay
animations [
 OnHover {
  binding_count=1
  length_seconds=0.1
  object=&self:OnHover
  sequence {bindings [IconContextButtonReplay_C_50 {
     id=#1
     possessed_class=IconContextButtonReplay_C
     tracks [{
       type=MovieSceneColorTrack
       display_name=Color and Opacity
       sections [{
         type=MovieSceneColorSection
         end_frame=2
         start_frame=0
         channels [
          Color.R {type=MovieSceneScriptingFloatChannel, const [1,2]}
          Color.G {type=MovieSceneScriptingFloatChannel, keys [[0,1],[2,0.413]]}
          Color.B {type=MovieSceneScriptingFloatChannel, keys [[0,1],[2,0]]}
          Color.A {type=MovieSceneScriptingFloatChannel, const [1,2]}
         ]
        }]
      }]
    }]}
 }
 OnSelected {
  binding_count=1
  length_seconds=0.1
  object=&self:OnSelected
  sequence {bindings [IconContextButtonReplay_C_50 {
     id=#1
     possessed_class=IconContextButtonReplay_C
     tracks [{
       type=MovieSceneColorTrack
       display_name=Color and Opacity
       sections [{
         type=MovieSceneColorSection
         end_frame=2
         start_frame=0
         channels [
          Color.R {type=MovieSceneScriptingFloatChannel, const [1,1]}
          Color.G {type=MovieSceneScriptingFloatChannel, const [0.413,1]}
          Color.B {type=MovieSceneScriptingFloatChannel, const [0,1]}
          Color.A {type=MovieSceneScriptingFloatChannel, const [1,1]}
         ]
        }]
      }]
    }]}
 }
]
graphs [
 EventGraph {nodes [
   n1 = "On Input Method Changed (InputActionWidget)"!() -> (OutputDelegate, then, bUsingGamepad:bool) {is=title K2Node_ComponentBoundEvent}
   > UpdateStyle!(execute, self, UsingGamepad:bool=@n1.bUsingGamepad) -> (then)
   Event On Hovered()
   > PlayAnimationForward(InAnimation=Get OnHover() -> (OnHover:Widget Animation Object Reference))
   n2 = Event On Released() {is=title K2Node_Event}
   PlayAnimationReverse(execute=[@n3.then, @n2.then], InAnimation=Get OnHover())
   n3 = Event On Unhovered()
   Event OnTriggeredInputActionChanged!() -> (OutputDelegate, then, NewTriggeredAction:"Data Table Row Handle Structure (by ref)") {is=title K2Node_Event}
   n4 = Get CommonInputSubsystem!() -> (ReturnValue:Common Input Object Reference) {is=title K2Node_GetBlueprintContextEx}
   > UpdateStyle(UsingGamepad="Equal (Enum)"(A=GetCurrentInputType(self=@n4), B=Gamepad))
   Event Construct()
   > InitializeButton()
   n5 = Get CommonInputSubsystem()
   > UpdateStyle(UsingGamepad="Equal (Enum)"(A=GetCurrentInputType(self=@n5), B=Gamepad))
   n6 = Event On Clicked()
   PlayAnimationForward(execute=[@n6.then, @n7.then], InAnimation=Get OnSelected() -> (OnSelected:Widget Animation Object Reference))
   n7 = Event On Pressed() {is=title K2Node_Event}
  ]}
 InitializeButton {nodes [
   InitializeButton() {is=title K2Node_FunctionEntry}
   n1 = Get Style() -> (Style:Common Button Style Class Reference)
   > n2 = Set MouseKeyboardStyle(MouseKeyboardStyle:Common Button Style Class Reference=@n1) -> (Output_Get:Common Button Style Class Reference)
   > SetControllerStyle!(execute, self, ControllerInputStyle:Common Button Style Class Reference=@n1) -> (then)
   > SetMouseKeyboardStyle!(execute, self, ControllerInputStyle:Common Button Style Class Reference=@n2) -> (then)
  ]}
 SetControllerStyle {nodes [
   n1 = SetControllerStyle!() -> (then, ControllerInputStyle:Common Button Style Class Reference) {is=title K2Node_FunctionEntry}
   > Set ControllerInputStyle(ControllerInputStyle:Common Button Style Class Reference=@n1) -> (Output_Get:Common Button Style Class Reference)
   n2 = Get CommonInputSubsystem()
   > UpdateStyle(UsingGamepad="Equal (Enum)"(A=GetCurrentInputType(self=@n2), B=Gamepad))
  ]}
 SetMouseKeyboardStyle {nodes [
   n1 = SetMouseKeyboardStyle!() -> (then, ControllerInputStyle:Common Button Style Class Reference) {is=title K2Node_FunctionEntry}
   > Set MouseKeyboardStyle(MouseKeyboardStyle=@n1)
   n2 = Get CommonInputSubsystem()
   > UpdateStyle(UsingGamepad="Equal (Enum)"(A=GetCurrentInputType(self=@n2), B=Gamepad))
  ]}
 UpdateStyle {nodes [
   n1 = UpdateStyle!() -> (then, UsingGamepad:bool=~) {is=title K2Node_FunctionEntry}
   > n2 = Branch(Condition=@n1)
   > SetStyle!(
    execute
    self:Common Button Base Object Reference
    InStyle:Common Button Style Class Reference=Get ControllerInputStyle() -> (ControllerInputStyle:Common Button Style Class Reference)
   ) -> (then)
   SetStyle(execute=@n2.else, InStyle=Get MouseKeyboardStyle() -> (MouseKeyboardStyle:Common Button Style Class Reference))
  ]}
]
variables [
 ControllerInputStyle {type=Class, default=null}
 DefaultColor {type=LinearColor, default=<rgba 0.0176 0.0232 0.0331 0.5>}
 HoveredColor {type=LinearColor, default=<rgba 1 0.4125 0 1>}
 MouseKeyboardStyle {type=Class, default=/FortUILibrary/Foundation/Buttons/ButtonStyles/Primary/ButtonStyle-Primary-M.ButtonStyle-Primary-M_C}
 SelectedColor {type=LinearColor, default=<rgba 1 0.4125 0 1>}
]
widgets {
 info {
  inheritedWidgetCount=0
  parentClass=/FortUILibrary/UIKit/Buttons/WBP_UIKit_Button_Generic.WBP_UIKit_Button_Generic_C
  rootWidgetClass=null
  widgetCount=0
 }
 widgets []
}
