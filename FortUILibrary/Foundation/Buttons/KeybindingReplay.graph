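# Replay UI input prompt widget pairing a title with keybind widgets, images, a controller image or a Common Action Widget.
# PreConstruct applies configured title, bindings, axis scales, images and height, resolving the action widget from the replay input DataTable.
# UpdateButtonVisibility picks the displayed variant from the switcher/image flags, gamepad input and Nintendo Switch platform.
# Construct calls it and binds to the global UI context's keybind-change event.
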WidgetBlueprint /FortUILibrary/Foundation/Buttons/KeybindingReplay
graphs [
 EventGraph {nodes [
   UpdateButtonVisibility() {is=title K2Node_CustomEvent}
   > n1 = Branch(Condition=Get bUseSwitcher() -> (bUseSwitcher:bool))
   > Branch(Condition=Get bUseImages() -> (bUseImages:bool))
   n2 = IsInputType!(InputType:ECommonInputType Enum=Gamepad, IsNot:bool=false) -> (Result:bool)
   > n3 = Branch(Condition=AND Boolean(A=Get bUseSpecificControllerImage() -> (bUseSpecificControllerImage:bool), B=@n2))
   SetActiveWidgetIndex(
    execute=[@n3.then, @n4.then]
    self:Common Visibility Switcher Object Reference=Get SwitcherMethod() -> (SwitcherMethod:Common Visibility Switcher Object Reference)
    Index=2
   )
   n4 = Branch(execute=@n3.else, Condition=IsNintendoSwitch!() -> (ReturnValue:bool))
   SetActiveWidgetIndex(execute=@n4.else, self=Get SwitcherMethod(), Index=Select(Option 0:int=1, Option 1:int=0, Index=@n2) -> (ReturnValue:int))
   n5 = Branch(execute=@n1.else, Condition=IsInputType(InputType=Gamepad, IsNot=false))
   > SetVisibility(self=Get SwitcherMethod(), InVisibility=Collapsed)
   > SetVisibility(self=Get ActionWidget() -> (ActionWidget:Common Action Widget Object Reference), InVisibility=Visible)
   SetVisibility(execute=@n5.else, self=Get SwitcherMethod(), InVisibility=SelfHitTestInvisible)
   > SetVisibility(self=Get ActionWidget(), InVisibility=Collapsed)
   > SetActiveWidgetIndex(self=Get SwitcherMethod(), Index=1)
   Event PreConstruct()
   > SetText(self=Get TitleText() -> (TitleText:Rich Text Block Object Reference), InText=Get Title() -> (Title:Text))
   > n6 = Branch(Condition=Get bUseSwitcher())
   > SetVisibility(self=Get SwitcherMethod(), InVisibility=SelfHitTestInvisible)
   > SetVisibility(self=Get ActionWidget(), InVisibility=Collapsed)
   > ShowKeybindWidgets()
   > SetupKeybindBoundActions()
   > n7 = SetupAxisScales()
   n8 = Branch(execute=[@n7.then, @n11.then], Condition=Get bUseImages())
   > ShowImages()
   > n9 = SetupImages()
   n10 = Branch(execute=[@n9.then, @n8.else], Condition=Get bUseCustomHeight() -> (bUseCustomHeight:bool))
   > SetHeightOverride(self=Get SizeBox() -> (SizeBox:Size Box Object Reference), InHeightOverride=Get CustomHeight() -> (CustomHeight:float))
   ClearHeightOverride(execute=@n10.else, self:Size Box Object Reference=Get SizeBox())
   SetVisibility(execute=@n6.else, self=Get SwitcherMethod(), InVisibility=Collapsed)
   > SetVisibility(self=Get ActionWidget(), InVisibility=Visible)
   > n11 = SetInputAction!(
    execute
    self:Common Action Widget Object Reference=Get ActionWidget()
    InputActionRow_DataTable:Data Table Object Reference=/FortUILibrary/Input/ReplayInputActionDataTable
    InputActionRow_RowName:Name=Get ActionWidgetBoundAction() -> (ActionWidgetBoundAction:Name=None)
   ) -> (then)
   Event Construct()
   > Bind Event to On Keybinds Changed(
    self:Fort Global UIContext Object Reference=Get FortGlobalUIContext!() -> (ReturnValue:Fort Global UIContext Object Reference) {is=title K2Node_GetBlueprintContextEx}
    Delegate=Create Event()
   ) {is=title K2Node_AddDelegate}
   > UpdateButtonVisibility() {is=title K2Node_CallFunction}
  ]}
 SetupAxisScales {nodes [
   SetupAxisScales() {is=title K2Node_FunctionEntry}
   > SetAxisScale!(
    execute
    self:Fort Keybind Widget Object Reference=Get KeybindWidget01() -> (KeybindWidget01:Keybind Widget Object Reference)
    NewValue:float=Get AxisScale01() -> (AxisScale01:float)
   ) -> (then)
   > SetAxisScale(self=Get KeybindWidget02() -> (KeybindWidget02:Keybind Widget Object Reference), NewValue=Get AxisScale02() -> (AxisScale02:float))
   > SetAxisScale(self=Get KeybindWidget03() -> (KeybindWidget03:Keybind Widget Object Reference), NewValue=Get AxisScale03() -> (AxisScale03:float))
   > SetAxisScale(self=Get KeybindWidget04() -> (KeybindWidget04:Keybind Widget Object Reference), NewValue=Get AxisScale04() -> (AxisScale04:float))
  ]}
 SetupImages {nodes [
   SetupImages() {is=title K2Node_FunctionEntry}
   > SetBrush!(
    execute
    self=Get Image01() -> (Image01:Image Object Reference)
    InBrush_TintColor:Slate Color Structure=(SpecifiedColor=(R=1,G=1,B=1,A=1),ColorUseRule=UseColor_Specified)
    InBrush_DrawAs:ESlateBrushDrawType Enum=Image
    InBrush_Tiling:ESlateBrushTileType Enum=NoTile
    InBrush_Mirroring:ESlateBrushMirrorType Enum=NoMirror
    InBrush_ImageSize:Vector 2D Structure=(X=32,Y=32)
    InBrush_Margin:Margin Structure=(Left=0,Top=0,Right=0,Bottom=0)
    InBrush_ResourceObject:Object Reference=Get SwitcherImage01() -> (SwitcherImage01:Object Reference)
    InBrush_OutlineSettings:Slate Brush Outline Settings Structure=(CornerRadii=(X=0,Y=0,Z=0,W=0),Color=(SpecifiedColor=(R=0,G=0,B=0,A=0),ColorUseRule=UseColor_Specified),Width=0,RoundingType=HalfHeightRadius,bUseBrushTransparency=False)
   ) -> (then)
   > SetBrush(
    self=Get Image02() -> (Image02:Image Object Reference)
    InBrush_TintColor=(SpecifiedColor=(R=1,G=1,B=1,A=1),ColorUseRule=UseColor_Specified)
    InBrush_DrawAs=Image
    InBrush_Tiling=NoTile
    InBrush_Mirroring=NoMirror
    InBrush_ImageSize=(X=32,Y=32)
    InBrush_Margin=(Left=0,Top=0,Right=0,Bottom=0)
    InBrush_ResourceObject=Get SwitcherImage02() -> (SwitcherImage02:Object Reference)
    InBrush_OutlineSettings=(CornerRadii=(X=0,Y=0,Z=0,W=0),Color=(SpecifiedColor=(R=0,G=0,B=0,A=0),ColorUseRule=UseColor_Specified),Width=0,RoundingType=HalfHeightRadius,bUseBrushTransparency=False)
   )
   > SetBrush(
    self=Get Image03() -> (Image03:Image Object Reference)
    InBrush_TintColor=(SpecifiedColor=(R=1,G=1,B=1,A=1),ColorUseRule=UseColor_Specified)
    InBrush_DrawAs=Image
    InBrush_Tiling=NoTile
    InBrush_Mirroring=NoMirror
    InBrush_ImageSize=(X=32,Y=32)
    InBrush_Margin=(Left=0,Top=0,Right=0,Bottom=0)
    InBrush_ResourceObject=Get SwitcherImage03() -> (SwitcherImage03:Object Reference)
    InBrush_OutlineSettings=(CornerRadii=(X=0,Y=0,Z=0,W=0),Color=(SpecifiedColor=(R=0,G=0,B=0,A=0),ColorUseRule=UseColor_Specified),Width=0,RoundingType=HalfHeightRadius,bUseBrushTransparency=False)
   )
   > SetBrush(
    self=Get Image04() -> (Image04:Image Object Reference)
    InBrush_TintColor=(SpecifiedColor=(R=1,G=1,B=1,A=1),ColorUseRule=UseColor_Specified)
    InBrush_DrawAs=Image
    InBrush_Tiling=NoTile
    InBrush_Mirroring=NoMirror
    InBrush_ImageSize=(X=32,Y=32)
    InBrush_Margin=(Left=0,Top=0,Right=0,Bottom=0)
    InBrush_ResourceObject=Get SwitcherImage04() -> (SwitcherImage04:Object Reference)
    InBrush_OutlineSettings=(CornerRadii=(X=0,Y=0,Z=0,W=0),Color=(SpecifiedColor=(R=0,G=0,B=0,A=0),ColorUseRule=UseColor_Specified),Width=0,RoundingType=HalfHeightRadius,bUseBrushTransparency=False)
   )
   > SetBrush(
    self=Get ControllerImage() -> (ControllerImage:Image Object Reference)
    InBrush_TintColor=(SpecifiedColor=(R=1,G=1,B=1,A=1),ColorUseRule=UseColor_Specified)
    InBrush_DrawAs=Image
    InBrush_Tiling=NoTile
    InBrush_Mirroring=NoMirror
    InBrush_ImageSize=(X=32,Y=32)
    InBrush_Margin=(Left=0,Top=0,Right=0,Bottom=0)
    InBrush_ResourceObject=Get ControllerBrush() -> (ControllerBrush:Object Reference)
    InBrush_OutlineSettings=(CornerRadii=(X=0,Y=0,Z=0,W=0),Color=(SpecifiedColor=(R=0,G=0,B=0,A=0),ColorUseRule=UseColor_Specified),Width=0,RoundingType=HalfHeightRadius,bUseBrushTransparency=False)
   )
  ]}
 SetupKeybindBoundActions {nodes [
   SetupKeybindBoundActions() {is=title K2Node_FunctionEntry}
   > SetBoundAction!(
    execute
    self:Fort Keybind Widget Object Reference=Get KeybindWidget01()
    NewBoundAction:Name=Get BoundActionName01() -> (BoundActionName01:Name=None)
   ) -> (then)
   > SetBoundAction(self=Get KeybindWidget02(), NewBoundAction=Get BoundActionName02() -> (BoundActionName02:Name=None))
   > SetBoundAction(self=Get KeybindWidget03(), NewBoundAction=Get BoundActionName03() -> (BoundActionName03:Name=None))
   > SetBoundAction(self=Get KeybindWidget04(), NewBoundAction=Get BoundActionName04() -> (BoundActionName04:Name=None))
  ]}
 ShowImages {nodes [
   ShowImages() {is=title K2Node_FunctionEntry}
   > n1 = Switch on Int!(execute:Exec, Selection:int=Get ImagesVisible() -> (ImagesVisible:int=0)) -> (Default:Exec, 0:Exec, 1:Exec, 2:Exec, 3:Exec, 4:Exec)
   SetVisibility(execute=[@n1.1, @n1.0], self=Get Image01(), InVisibility=SelfHitTestInvisible)
   > SetVisibility(self=[Get Image02(), Get Image04(), Get Image03()], InVisibility=Collapsed)
   SetVisibility(execute=@n1.4, self=[Get Image02(), Get Image04(), Get Image03(), Get Image01()], InVisibility=SelfHitTestInvisible)
   SetVisibility(execute=@n1.3, self=[Get Image01(), Get Image02(), Get Image03()], InVisibility=SelfHitTestInvisible)
   > SetVisibility(self=Get Image04(), InVisibility=Collapsed)
   SetVisibility(execute=@n1.2, self=[Get Image01(), Get Image02()], InVisibility=SelfHitTestInvisible)
   > SetVisibility(self=[Get Image04(), Get Image03()], InVisibility=Collapsed)
  ]}
 ShowKeybindWidgets {nodes [
   ShowKeybindWidgets() {is=title K2Node_FunctionEntry}
   > n1 = Switch on Int(Selection=Get KeybindsVisible() -> (KeybindsVisible:int=0))
   SetVisibility(execute=[@n1.1, @n1.0], self=Get KeybindWidget01(), InVisibility=SelfHitTestInvisible)
   > SetVisibility(self=[Get KeybindWidget02(), Get KeybindWidget03(), Get KeybindWidget04()], InVisibility=Collapsed)
   SetVisibility(
    execute=@n1.4
    self=[Get KeybindWidget02(), Get KeybindWidget04(), Get KeybindWidget03(), Get KeybindWidget01()]
    InVisibility=SelfHitTestInvisible
   )
   SetVisibility(execute=@n1.3, self=[Get KeybindWidget03(), Get KeybindWidget01(), Get KeybindWidget02()], InVisibility=SelfHitTestInvisible)
   > SetVisibility(self=Get KeybindWidget04(), InVisibility=Collapsed)
   SetVisibility(execute=@n1.2, self=[Get KeybindWidget01(), Get KeybindWidget02()], InVisibility=SelfHitTestInvisible)
   > SetVisibility(self=[Get KeybindWidget03(), Get KeybindWidget04()], InVisibility=Collapsed)
  ]}
 UpdateKeybindText {nodes [
   n1 = UpdateKeybindText!() -> (then, NewText:"Text (by ref)") {is=title K2Node_FunctionEntry}
   > SetText(self=Get TitleText(), InText=@n1)
  ]}
]
variables [
 ActionWidgetBoundAction {type=string, default=None}
 AxisScale01 {type=number}
 AxisScale02 {type=number}
 AxisScale03 {type=number}
 AxisScale04 {type=number}
 BoundActionName01 {type=string, default=None}
 BoundActionName02 {type=string, default=None}
 BoundActionName03 {type=string, default=None}
 BoundActionName04 {type=string, default=None}
 bUseCustomHeight {type=boolean}
 bUseImages {type=boolean}
 bUseSpecificControllerImage {type=boolean}
 bUseSwitcher {type=boolean, default=true}
 ControllerBrush {type=Object, default=null}
 CustomHeight {type=number}
 ImagesVisible {type=integer}
 KeybindsVisible {type=integer, default=1}
 SwitcherImage01 {type=Object, default=null}
 SwitcherImage02 {type=Object, default=null}
 SwitcherImage03 {type=Object, default=null}
 SwitcherImage04 {type=Object, default=null}
 Title {type=string, default="{Input Action}"}
]
widgets {
 info {inheritedWidgetCount=0, parentClass=&UserWidget, rootWidgetClass=&SizeBox, widgetCount=19}
 widgets [
  SizeBox {widgetClassPath=&SizeBox, bIsVariable=true, widget=SizeBox, properties {minDesiredWidth=260, widthOverride=300}}
  Main {
   widgetClassPath=&Overlay
   parent=SizeBox
   widget=Main
   properties {bOverride_Cursor=true, visibility=Visible}
   slot {class=SizeBoxSlot, properties {verticalAlignment=VAlign_Center}}
  }
  Backplate {
   widgetClassPath=&WBP_UIKit_Backplate
   parent=Main
   widget=Backplate
   properties {
    backplateMaterial=/FortUILibrary/Foundation/Buttons/Materials/M_Timeline_Backplate_Inst
    brightness=Medium
    cornerRadius=Tab
    backplateSize=<0 42>
   }
   slot {class=OverlaySlot, properties {horizontalAlignment=HAlign_Fill, verticalAlignment=VAlign_Fill}}
  }
  MasterHorizontalBox {
   widgetClassPath=&HorizontalBox
   parent=Main
   widget=MasterHorizontalBox
   slot {class=OverlaySlot, properties {horizontalAlignment=HAlign_Center, verticalAlignment=VAlign_Center, padding=<ltrb 7 7 7 7>}}
  }
  SwitcherMethod {
   widgetClassPath=&CommonVisibilitySwitcher
   bIsVariable=true
   parent=MasterHorizontalBox
   widget=SwitcherMethod
   slot {class=HorizontalBoxSlot}
  }
  KeybindWidgetBox {
   widgetClassPath=&HorizontalBox
   parent=SwitcherMethod
   widget=KeybindWidgetBox
   slot {class=CommonVisibilitySwitcherSlot, properties {horizontalAlignment=HAlign_Center, verticalAlignment=VAlign_Center}}
  }
  KeybindWidget01 {
   widgetClassPath=/FortUILibrary/Input/KeybindWidget
   bIsVariable=true
   parent=KeybindWidgetBox
   widget=KeybindWidget01
   slot {class=HorizontalBoxSlot}
  }
  KeybindWidget02 {
   widgetClassPath=&KeybindWidget
   bIsVariable=true
   parent=KeybindWidgetBox
   widget=KeybindWidget02
   slot {class=HorizontalBoxSlot, properties {padding=<ltrb 5 0 0 0>}}
  }
  KeybindWidget03 {
   widgetClassPath=&KeybindWidget
   bIsVariable=true
   parent=KeybindWidgetBox
   widget=KeybindWidget03
   slot {class=HorizontalBoxSlot, properties {padding=<ltrb 5 0 0 0>}}
  }
  KeybindWidget04 {
   widgetClassPath=&KeybindWidget
   bIsVariable=true
   parent=KeybindWidgetBox
   widget=KeybindWidget04
   slot {class=HorizontalBoxSlot, properties {padding=<ltrb 5 0 0 0>}}
  }
  ImageHorizontalBox {
   widgetClassPath=&HorizontalBox
   parent=SwitcherMethod
   widget=ImageHorizontalBox
   slot {class=CommonVisibilitySwitcherSlot, properties {horizontalAlignment=HAlign_Center, verticalAlignment=VAlign_Center}}
  }
  Image01 {
   widgetClassPath=&Image
   bIsVariable=true
   parent=ImageHorizontalBox
   widget=Image01
   properties {brush {
     bIsDynamicallyLoaded=false
     drawAs=Image
     imageType=FullColor
     mirroring=NoMirror
     resourceName=None
     resourceObject=/FortUILibrary/Input/Mouse/MaterialInstances/MI_UI_Input_Mouse_LeftClick
     tiling=NoTile
     imageSize=<32 32>
     margin=<ltrb 0 0 0 0>
     outlineSettings {
      bUseBrushTransparency=false
      roundingType=HalfHeightRadius
      width=0
      color {colorUseRule=UseColor_Specified, specifiedColor=<rgba 0 0 0 0>}
      cornerRadii=<0 0 0 0>
     }
     tintColor {colorUseRule=UseColor_Specified, specifiedColor=<rgba 1 1 1 1>}
     uVRegion {bIsValid=false, max=<0 0>, min=<0 0>}
    }}
   slot {class=HorizontalBoxSlot}
  }
  Image02 {
   widgetClassPath=&Image
   bIsVariable=true
   parent=ImageHorizontalBox
   widget=Image02
   properties {brush {
     bIsDynamicallyLoaded=false
     drawAs=Image
     imageType=FullColor
     mirroring=NoMirror
     resourceName=None
     resourceObject=&MI_UI_Input_Mouse_LeftClick
     tiling=NoTile
     imageSize=<32 32>
     margin=<ltrb 0 0 0 0>
     outlineSettings {
      bUseBrushTransparency=false
      roundingType=HalfHeightRadius
      width=0
      color {colorUseRule=UseColor_Specified, specifiedColor=<rgba 0 0 0 0>}
      cornerRadii=<0 0 0 0>
     }
     tintColor {colorUseRule=UseColor_Specified, specifiedColor=<rgba 1 1 1 1>}
     uVRegion {bIsValid=false, max=<0 0>, min=<0 0>}
    }}
   slot {class=HorizontalBoxSlot}
  }
  Image03 {
   widgetClassPath=&Image
   bIsVariable=true
   parent=ImageHorizontalBox
   widget=Image03
   properties {brush {
     bIsDynamicallyLoaded=false
     drawAs=Image
     imageType=FullColor
     mirroring=NoMirror
     resourceName=None
     resourceObject=&MI_UI_Input_Mouse_LeftClick
     tiling=NoTile
     imageSize=<32 32>
     margin=<ltrb 0 0 0 0>
     outlineSettings {
      bUseBrushTransparency=false
      roundingType=HalfHeightRadius
      width=0
      color {colorUseRule=UseColor_Specified, specifiedColor=<rgba 0 0 0 0>}
      cornerRadii=<0 0 0 0>
     }
     tintColor {colorUseRule=UseColor_Specified, specifiedColor=<rgba 1 1 1 1>}
     uVRegion {bIsValid=false, max=<0 0>, min=<0 0>}
    }}
   slot {class=HorizontalBoxSlot}
  }
  Image04 {
   widgetClassPath=&Image
   bIsVariable=true
   parent=ImageHorizontalBox
   widget=Image04
   properties {brush {
     bIsDynamicallyLoaded=false
     drawAs=Image
     imageType=FullColor
     mirroring=NoMirror
     resourceName=None
     resourceObject=&MI_UI_Input_Mouse_LeftClick
     tiling=NoTile
     imageSize=<32 32>
     margin=<ltrb 0 0 0 0>
     outlineSettings {
      bUseBrushTransparency=false
      roundingType=HalfHeightRadius
      width=0
      color {colorUseRule=UseColor_Specified, specifiedColor=<rgba 0 0 0 0>}
      cornerRadii=<0 0 0 0>
     }
     tintColor {colorUseRule=UseColor_Specified, specifiedColor=<rgba 1 1 1 1>}
     uVRegion {bIsValid=false, max=<0 0>, min=<0 0>}
    }}
   slot {class=HorizontalBoxSlot}
  }
  ControllerImage {
   widgetClassPath=&Image
   bIsVariable=true
   parent=SwitcherMethod
   widget=ControllerImage
   properties {brush {
     bIsDynamicallyLoaded=false
     drawAs=Image
     imageType=FullColor
     mirroring=NoMirror
     resourceName=None
     resourceObject=/FortUILibrary/Input/GamepadPS4/Source/T_PS4_LStick
     tiling=NoTile
     imageSize=<32 32>
     margin=<ltrb 0 0 0 0>
     outlineSettings {
      bUseBrushTransparency=false
      roundingType=HalfHeightRadius
      width=0
      color {colorUseRule=UseColor_Specified, specifiedColor=<rgba 0 0 0 0>}
      cornerRadii=<0 0 0 0>
     }
     tintColor {colorUseRule=UseColor_Specified, specifiedColor=<rgba 1 1 1 1>}
     uVRegion {bIsValid=false, max=<0 0>, min=<0 0>}
    }}
   slot {class=CommonVisibilitySwitcherSlot}
  }
  ActionWidget {
   widgetClassPath=&CommonActionWidget
   bIsVariable=true
   parent=MasterHorizontalBox
   widget=ActionWidget
   properties {progressMaterialBrush {
     bIsDynamicallyLoaded=false
     drawAs=Image
     imageType=FullColor
     mirroring=NoMirror
     resourceName=None
     resourceObject=/FortUILibrary/Input/Misc/T_Preview_Button
     tiling=NoTile
     imageSize=<32 32>
     margin=<ltrb 0 0 0 0>
     outlineSettings {
      bUseBrushTransparency=false
      roundingType=HalfHeightRadius
      width=0
      color {colorUseRule=UseColor_Specified, specifiedColor=<rgba 0 0 0 0>}
      cornerRadii=<0 0 0 0>
     }
     tintColor {colorUseRule=UseColor_Specified, specifiedColor=<rgba 1 1 1 1>}
     uVRegion {bIsValid=false, max=<0 0>, min=<0 0>}
    }}
   slot {class=HorizontalBoxSlot}
  }
  ActionWidgetMethod {
   widgetClassPath=&CommonBorder
   bIsVariable=true
   parent=MasterHorizontalBox
   widget=ActionWidgetMethod
   properties {padding=<ltrb 5 0 5 0>}
   slot {class=HorizontalBoxSlot, properties {verticalAlignment=VAlign_Center}}
  }
  TitleText {
   widgetClassPath=/Script/UMG.RichTextBlock
   bIsVariable=true
   parent=ActionWidgetMethod
   widget=TitleText
   properties {
    bOverrideDefaultStyle=true
    justification=Right
    text="{Input Action}"
    defaultTextStyleOverride {
     overflowPolicy=Clip
     transformPolicy=None
     colorAndOpacity {colorUseRule=UseColor_Specified, specifiedColor=<rgba 1 1 1 1>}
     font {
      bForceMonospaced=false
      bMaterialIsStencil=false
      fontMaterial=null
      fontObject=&HeadingNow
      letterSpacing=0
      monospacedWidth=1
      size=13.5
      skewAmount=0
      typefaceFontName=74 Regular
      outlineSettings {
       bApplyOutlineToDropShadows=false
       bMiteredCorners=false
       bSeparateFillAlpha=false
       outlineMaterial=null
       outlineSize=0
       outlineColor=<rgba 0 0 0 1>
      }
     }
     highlightColor {colorUseRule=UseColor_Specified, specifiedColor=<rgba 0 0 0 1>}
     highlightShape {
      bIsDynamicallyLoaded=false
      drawAs=Image
      imageType=NoImage
      mirroring=NoMirror
      resourceName=None
      resourceObject=null
      tiling=NoTile
      imageSize=<32 32>
      margin=<ltrb 0 0 0 0>
      outlineSettings {
       bUseBrushTransparency=false
       roundingType=HalfHeightRadius
       width=0
       color {colorUseRule=UseColor_Specified, specifiedColor=<rgba 0 0 0 0>}
       cornerRadii=<0 0 0 0>
      }
      tintColor {colorUseRule=UseColor_Specified, specifiedColor=<rgba 1 1 1 1>}
      uVRegion {bIsValid=false, max=<0 0>, min=<0 0>}
     }
     selectedBackgroundColor {colorUseRule=UseColor_Foreground, specifiedColor=<rgba 1 0 1 1>}
     shadowColorAndOpacity=<rgba 0 0 0 1>
     shadowOffset=<0 0>
     strikeBrush {
      bIsDynamicallyLoaded=false
      drawAs=Image
      imageType=NoImage
      mirroring=NoMirror
      resourceName=None
      resourceObject=null
      tiling=NoTile
      imageSize=<32 32>
      margin=<ltrb 0 0 0 0>
      outlineSettings {
       bUseBrushTransparency=false
       roundingType=HalfHeightRadius
       width=0
       color {colorUseRule=UseColor_Specified, specifiedColor=<rgba 0 0 0 0>}
       cornerRadii=<0 0 0 0>
      }
      tintColor {colorUseRule=UseColor_Specified, specifiedColor=<rgba 1 1 1 1>}
      uVRegion {bIsValid=false, max=<0 0>, min=<0 0>}
     }
     underlineBrush {
      bIsDynamicallyLoaded=false
      drawAs=Image
      imageType=NoImage
      mirroring=NoMirror
      resourceName=None
      resourceObject=null
      tiling=NoTile
      imageSize=<32 32>
      margin=<ltrb 0 0 0 0>
      outlineSettings {
       bUseBrushTransparency=false
       roundingType=HalfHeightRadius
       width=0
       color {colorUseRule=UseColor_Specified, specifiedColor=<rgba 0 0 0 0>}
       cornerRadii=<0 0 0 0>
      }
      tintColor {colorUseRule=UseColor_Specified, specifiedColor=<rgba 1 1 1 1>}
      uVRegion {bIsValid=false, max=<0 0>, min=<0 0>}
     }
    }
   }
   slot {class=BorderSlot, properties {padding=<ltrb 5 0 5 0>}}
  }
 ]
}
