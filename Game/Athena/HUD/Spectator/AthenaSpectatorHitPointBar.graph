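# Spectator HUD widget that shows a player's health, shield or vehicle health as a layered fill and delta bar with optional icon and numbers.
# InitializeBar picks colors and icon by BarType, on PreConstruct with design-time preview values.
# Value changes set fill progress and update text from PlayerData, scaling normalized health and shield by their maximum.
# When downed (DBNO), the health fill takes the delta color and the shield bar can hide.
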WidgetBlueprint /Game/Athena/HUD/Spectator/AthenaSpectatorHitPointBar
graphs [
 EventGraph {nodes [
   n1 = Event PreConstruct()
   > n2 = Branch(Condition=@n1.IsDesignTime)
   > OnCurrentValueChanged!(
    execute
    self:Athena Spectator Hit Point Bar Base Object Reference
    InCurrentValue:float=Get DesignTime_CurrentValue() -> (DesignTime_CurrentValue:double)
   ) -> (then)
   > n3 = OnDeltaValueChanged!(
    execute
    self:Athena Spectator Hit Point Bar Base Object Reference
    InDeltaValue:float=Get DesignTime_DeltaValue() -> (DesignTime_DeltaValue:double)
   ) -> (then)
   InitializeBar(execute=[@n3.then, @n2.else])
   > SetVisibility(
    self=Get NumbersHB() -> (NumbersHB:Horizontal Box Object Reference)
    InVisibility=Select(Option 0=Collapsed, Option 1=SelfHitTestInvisible, Index=Get ShowNumbers() -> (ShowNumbers:bool))
   )
   > UpdateMaxValueText()
   > n4 = Branch(Condition="Equal (Enum)"(A:EHealthBarType Enum=Get BarType() -> (BarType:EHealthBarType Enum=Health), B:EHealthBarType Enum=Health))
   > SetBrushFromTexture!(
    execute
    self:Image Object Reference=Get Icon() -> (Icon:Image Object Reference)
    Texture:Texture 2D Object Reference=Get HealthTexture() -> (HealthTexture:Texture 2D Object Reference)
    bMatchSize:bool=true
   ) -> (then)
   SetBrushFromAtlasInterface!(
    execute=@n4.else
    self:Image Object Reference=Get Icon()
    AtlasRegion:Slate Texture Atlas Interface Interface=Select!(
     Health:Slate Texture Atlas Interface Interface
     Shield:Slate Texture Atlas Interface Interface=Get ShieldIcon() -> (ShieldIcon:Paper Sprite Object Reference)
     Overshield:Slate Texture Atlas Interface Interface
     Armor:Slate Texture Atlas Interface Interface
     Stamina:Slate Texture Atlas Interface Interface
     VehicleHealth:Slate Texture Atlas Interface Interface=Get VehicleIcon() -> (VehicleIcon:Paper Sprite Object Reference)
     SignalInStorm:Slate Texture Atlas Interface Interface
     Index:EHealthBarType Enum=Get BarType() -> (BarType=Health)
    ) -> (ReturnValue:Slate Texture Atlas Interface Interface)
    bMatchSize:bool=false
   ) -> (then)
   n5 = Event ChangeTextVisibility!() -> (OutputDelegate, then, bIsVisible:bool) {is=title K2Node_Event}
   n6 = Select!(Option 0:ESlateVisibility Enum=Collapsed, Option 1:ESlateVisibility Enum=Visible, Index:bool=@n5.bIsVisible) -> (ReturnValue:ESlateVisibility Enum)
   > SetVisibility(self=Get CurrentValue() -> (CurrentValue:Text Object Reference), InVisibility=@n6)
   > SetVisibility(self=Get MaxValue() -> (MaxValue:Text Object Reference), InVisibility=@n6)
   n7 = Event OnCurrentValueChanged!() -> (OutputDelegate, then, InCurrentValue:float) {is=title K2Node_Event}
   > SetProgress!(
    execute
    self:Fort Simple Material Progress Bar Object Reference=Get ProgressBarFill() -> (ProgressBarFill:Simple Material Progress Bar Object Reference)
    Progress:float=@n7.InCurrentValue
   ) -> (then)
   > UpdateMaxValueText()
   > UpdateCurrentValueText!(execute, self, InCurrentValue:double=@n7.InCurrentValue) -> (then)
   n8 = Event OnDBNOChanged!() -> (OutputDelegate, then, bInIsDBNO:bool) {is=title K2Node_Event}
   > n9 = Switch on EHealthBarType!(execute, Selection:EHealthBarType Enum=Get BarType() -> (BarType=Health)) -> (Health:Exec, Shield:Exec, Overshield:Exec, Armor:Exec, Stamina:Exec, VehicleHealth:Exec, SignalInStorm:Exec) {is=title K2Node_SwitchEnum}
   > n10 = Branch(Condition=@n8.bInIsDBNO)
   n11 = Get ProgressBarFill()
   n12 = Get HealthDeltaColorB() -> (HealthDeltaColorB:Linear Color Structure)
   > SetColorA!(execute, self:Fort Simple Material Progress Bar Object Reference=@n11, ColorA:Linear Color Structure=@n12) -> (then)
   > SetColorB!(execute, self:Fort Simple Material Progress Bar Object Reference=@n11, ColorB:Linear Color Structure=@n12) -> (then)
   InitializeBar(execute=@n10.else)
   n13 = Get HideShieldOnDbno() -> (HideShieldOnDbno:bool)
   SetVisibility(execute=@n9.Shield, InVisibility=Select(Option 0=SelfHitTestInvisible, Option 1=Hidden, Index=AND Boolean(A=@n8.bInIsDBNO, B=@n13)))
  ]}
 InitializeBar {nodes [
   InitializeBar() {is=title K2Node_FunctionEntry}
   n1 = Get BarType() -> (BarType=Health)
   n2 = Get ProgressBarFill()
   > SetColorA(
    self=@n2
    ColorA=Select!(
     Health:Linear Color Structure=Get HealthFillColorA() -> (HealthFillColorA:Linear Color Structure)
     Shield:Linear Color Structure=Get ShieldFillColorA() -> (ShieldFillColorA:Linear Color Structure)
     Overshield:Linear Color Structure
     Armor:Linear Color Structure
     Stamina:Linear Color Structure=(R=0,G=0,B=0,A=0)
     VehicleHealth:Linear Color Structure=Get VehicleHealthFillColorA() -> (VehicleHealthFillColorA:Linear Color Structure)
     SignalInStorm:Linear Color Structure
     Index:EHealthBarType Enum=@n1
    ) -> (ReturnValue:Linear Color Structure)
   )
   > SetColorB(
    self=@n2
    ColorB=Select(
     Health=Get HealthFillColorB() -> (HealthFillColorB:Linear Color Structure)
     Shield=Get ShieldFillColorB() -> (ShieldFillColorB:Linear Color Structure)
     Stamina=(R=0,G=0,B=0,A=0)
     VehicleHealth=Get VehicleHealthFillColorB() -> (VehicleHealthFillColorB:Linear Color Structure)
     Index=@n1
    )
   )
   n3 = Get ProgressBarDelta() -> (ProgressBarDelta:Fort Simple Material Progress Bar Object Reference)
   > SetColorA(
    self=@n3
    ColorA=Select(
     Health=Get HealthDeltaColorA() -> (HealthDeltaColorA:Linear Color Structure)
     Shield=Get ShieldDeltaColorA() -> (ShieldDeltaColorA:Linear Color Structure)
     Stamina=(R=0,G=0,B=0,A=0)
     VehicleHealth=Get VehicleHealthDeltaColorA() -> (VehicleHealthDeltaColorA:Linear Color Structure)
     Index=@n1
    )
   )
   > SetColorB(
    self=@n3
    ColorB=Select(
     Health=Get HealthDeltaColorB()
     Shield=Get ShieldDeltaColorB() -> (ShieldDeltaColorB:Linear Color Structure)
     Stamina=(R=0,G=0,B=0,A=0)
     VehicleHealth=Get VehicleHealthDeltaColorB() -> (VehicleHealthDeltaColorB:Linear Color Structure)
     Index=@n1
    )
   )
   > SetColorBackground!(
    execute
    self:Fort Simple Material Progress Bar Object Reference=@n3
    ColorBackground:Linear Color Structure=Select(
     Health=Get HealthBackgroundColor() -> (HealthBackgroundColor:Linear Color Structure)
     Shield=Get ShieldBackgroundColor() -> (ShieldBackgroundColor:Linear Color Structure)
     Stamina=(R=0,G=0,B=0,A=0)
     VehicleHealth=Get VehicleHealthBackgroundColor() -> (VehicleHealthBackgroundColor:Linear Color Structure)
     Index=@n1
    )
   ) -> (then)
  ]}
 UpdateCurrentValueText {nodes [
   n1 = UpdateCurrentValueText!() -> (then, InCurrentValue:double=~) {is=title K2Node_FunctionEntry}
   n2 = GetVehicleHealth!(self:Athena Remote Player View Data Object Reference=Get PlayerData() -> (PlayerData:Athena Remote Player View Data Weak object)) -> (ReturnValue:float)
   n3 = float * float(A=@n1, B=Get MaxHealthValue() -> (MaxHealthValue:double))
   n4 = Get BarType() -> (BarType=Health)
   > "SetText (Text)"(
    self=Get CurrentValue()
    InText="To Text (Float)"(
     Value=Select(
      Health:double=@n3
      Shield:double=@n3
      Overshield:double=0
      Armor:double=0
      Stamina:double=0
      VehicleHealth:double=@n2
      SignalInStorm:double=0
      Index=@n4
     ) -> (ReturnValue:double)
     MinimumIntegralDigits=1
     MaximumIntegralDigits=324
     MinimumFractionalDigits=0
     MaximumFractionalDigits=0
    )
   )
  ]}
 UpdateMaxValueText {nodes [
   UpdateMaxValueText() {is=title K2Node_FunctionEntry}
   > n1 = Get!(execute:Exec) -> (then:Exec, else:Exec, PlayerData:Athena Remote Player View Data Weak object)
   > n2 = Set MaxHealthValue(MaxHealthValue:double=Select(
     Health=GetHealth!(self:Athena Remote Player View Data Object Reference=@n1) -> (Health:float, HealthMax:float).HealthMax
     Shield=GetShield!(self:Athena Remote Player View Data Object Reference=@n1) -> (Shield:float, ShieldMax:float).ShieldMax
     Overshield=0
     Armor=0
     Stamina=100
     VehicleHealth=GetVehicleMaxHealth!(self:Athena Remote Player View Data Object Reference=@n1) -> (ReturnValue:float)
     SignalInStorm=0
     Index=Get BarType() -> (BarType=Health)
    )) -> (Output_Get:double)
   > "SetText (Text)"(
    self=Get MaxValue()
    InText="To Text (Float)"(Value=@n2, MinimumIntegralDigits=1, MaximumIntegralDigits=324, MinimumFractionalDigits=0, MaximumFractionalDigits=0)
   )
  ]}
]
variables [
 DesignTime_CurrentValue {type=number, default=0.5}
 DesignTime_DeltaValue {type=number, default=0.75}
 HealthBackgroundColor {type=LinearColor, default=<rgba 0 0.0012 0.0103 0.6>}
 HealthDeltaColorA {type=LinearColor, default=<rgba 1 0.1647 0.1647 1>}
 HealthDeltaColorB {type=LinearColor, default=<rgba 0.7843 0.051 0.051 1>}
 HealthFillColorA {type=LinearColor, default=<rgba 0.0745 0.502 0.0353 1>}
 HealthFillColorB {type=LinearColor, default=<rgba 0.1765 0.7294 0.0784 1>}
 HealthTexture {type=Texture2D, default=/Game/Athena/HUD/Spectator/Art/T-Icon-Broadcast-Healthbar-S}
 HideShieldOnDbno {type=boolean, default=true}
 MaxHealthValue {type=number, default=100}
 ShieldBackgroundColor {type=LinearColor, default=<rgba 0 0.0012 0.0103 0.6>}
 ShieldDeltaColorA {type=LinearColor, default=<rgba 0.2039 0.6863 0.9569 1>}
 ShieldDeltaColorB {type=LinearColor, default=<rgba 0.0863 0.4392 0.9137 1>}
 ShieldFillColorA {type=LinearColor, default=<rgba 0.0353 0.1804 0.8314 1>}
 ShieldFillColorB {type=LinearColor, default=<rgba 0.0863 0.4392 0.9137 1>}
 ShieldIcon {type=PaperSprite, default=/Game/Athena/HUD/Atlas/Shared/icon_shield_Sprite}
 ShowNumbers {type=boolean}
 VehicleHealthBackgroundColor {type=LinearColor, default=<rgba 0 0.0012 0.0103 0.6>}
 VehicleHealthDeltaColorA {type=LinearColor, default=<rgba 1 0.4852 0.147 1>}
 VehicleHealthDeltaColorB {type=LinearColor, default=<rgba 0.855 0.3763 0.0482 1>}
 VehicleHealthFillColorA {type=LinearColor, default=<rgba 0.5457 0.2159 0 1>}
 VehicleHealthFillColorB {type=LinearColor, default=<rgba 0.7605 0.3515 0.0823 1>}
 VehicleIcon {type=PaperSprite, default=/Game/Athena/HUD/Atlas/Shared/icon_vehiclehealth_sprite.icon_vehiclehealth_Sprite}
]
widgets {
 info {inheritedWidgetCount=1, parentClass=/Script/FortniteUI.AthenaSpectatorHitPointBarBase, rootWidgetClass=&Overlay, widgetCount=13}
 widgets [
  Overlay_0 {widgetClassPath=&Overlay, widget=Overlay_0}
  ProgressBarDelta {
   widgetClassPath=/Game/Athena/HUD/SimpleMaterialProgressBar
   bInherited=true
   bIsVariable=true
   parent=Overlay_0
   widget=ProgressBarDelta
   properties {designTime_Progress=0.75}
   slot {class=OverlaySlot, properties {horizontalAlignment=HAlign_Fill, verticalAlignment=VAlign_Fill}}
  }
  ProgressBarFill {
   widgetClassPath=&SimpleMaterialProgressBar
   bIsVariable=true
   parent=Overlay_0
   widget=ProgressBarFill
   properties {designTime_Progress=0.5}
   slot {class=OverlaySlot, properties {horizontalAlignment=HAlign_Fill, verticalAlignment=VAlign_Fill}}
  }
  NumbersHB {
   widgetClassPath=&HorizontalBox
   bIsVariable=true
   parent=Overlay_0
   widget=NumbersHB
   slot {class=OverlaySlot, properties {horizontalAlignment=HAlign_Fill, verticalAlignment=VAlign_Center}}
  }
  Spacer_2 {widgetClassPath=&Spacer, parent=NumbersHB, widget=Spacer_2, properties {size=<4 0>}, slot {class=HorizontalBoxSlot}}
  SizeBox_0 {
   widgetClassPath=&SizeBox
   parent=NumbersHB
   widget=SizeBox_0
   properties {bOverride_HeightOverride=true, bOverride_WidthOverride=true, heightOverride=24, maxDesiredWidth=12, widthOverride=24}
   slot {class=HorizontalBoxSlot, properties {horizontalAlignment=HAlign_Center, verticalAlignment=VAlign_Center}}
  }
  Icon {
   widgetClassPath=&Image
   bIsVariable=true
   parent=SizeBox_0
   widget=Icon
   properties {brush {
     bIsDynamicallyLoaded=false
     drawAs=Image
     imageType=NoImage
     mirroring=NoMirror
     resourceName=None
     resourceObject=&T-Icon-Broadcast-Healthbar-S
     tiling=NoTile
     imageSize=<24 24>
     margin=<ltrb 0 0 0 0>
     outlineSettings {
      bUseBrushTransparency=false
      roundingType=HalfHeightRadius
      width=0
      color {colorUseRule=UseColor_Specified, specifiedColor=<rgba 0 0 0 0>}
      cornerRadii=<0 0 0 0>
     }
     tintColor {colorUseRule=UseColor_Specified, specifiedColor=<rgba 1 1 1 1>}
     uVRegion {bIsValid=false, max=<0 0>, min=<0 0>}
    }}
   slot {class=SizeBoxSlot}
  }
  Spacer_3 {widgetClassPath=&Spacer, parent=NumbersHB, widget=Spacer_3, properties {size=<4 0>}, slot {class=HorizontalBoxSlot}}
  CurrentValue {
   widgetClassPath=&TextBlock
   bIsVariable=true
   parent=NumbersHB
   widget=CurrentValue
   properties {
    text=888
    font {
     bForceMonospaced=false
     bMaterialIsStencil=false
     fontMaterial=null
     fontObject=&NotoSans
     letterSpacing=0
     monospacedWidth=1
     size=16
     skewAmount=0
     typefaceFontName=NotoSans-Bold
     outlineSettings {
      bApplyOutlineToDropShadows=false
      bMiteredCorners=false
      bSeparateFillAlpha=false
      outlineMaterial=null
      outlineSize=0
      outlineColor=<rgba 0 0 0 1>
     }
    }
    renderTransform {angle=0, scale=<1 1>, shear=<0 0>, translation=<0 -2>}
   }
   slot {class=HorizontalBoxSlot, properties {verticalAlignment=VAlign_Center}}
  }
  Spacer_4 {widgetClassPath=&Spacer, parent=NumbersHB, widget=Spacer_4, properties {size=<4 0>}, slot {class=HorizontalBoxSlot}}
  Split {
   widgetClassPath=&Image
   parent=NumbersHB
   widget=Split
   properties {
    brush {
     bIsDynamicallyLoaded=false
     drawAs=Image
     imageType=NoImage
     mirroring=NoMirror
     resourceName=None
     resourceObject=null
     tiling=NoTile
     imageSize=<2 16>
     margin=<ltrb 0 0 0 0>
     outlineSettings {
      bUseBrushTransparency=false
      roundingType=HalfHeightRadius
      width=0
      color {colorUseRule=UseColor_Specified, specifiedColor=<rgba 0 0 0 0>}
      cornerRadii=<0 0 0 0>
     }
     tintColor {colorUseRule=UseColor_Specified, specifiedColor=<rgba 1 1 1 1>}
     uVRegion {bIsValid=false, max=<0 0>, min=<0 0>}
    }
    colorAndOpacity=<rgba 1 1 1 0.7>
   }
   slot {class=HorizontalBoxSlot, properties {verticalAlignment=VAlign_Center}}
  }
  Spacer_5 {widgetClassPath=&Spacer, parent=NumbersHB, widget=Spacer_5, properties {size=<4 0>}, slot {class=HorizontalBoxSlot}}
  MaxValue {
   widgetClassPath=&TextBlock
   bIsVariable=true
   parent=NumbersHB
   widget=MaxValue
   properties {
    text=888
    colorAndOpacity {colorUseRule=UseColor_Specified, specifiedColor=<rgba 1 1 1 0.75>}
    font {
     bForceMonospaced=false
     bMaterialIsStencil=false
     fontMaterial=null
     fontObject=&NotoSans
     letterSpacing=0
     monospacedWidth=1
     size=16
     skewAmount=0
     typefaceFontName=NotoSans-Bold
     outlineSettings {
      bApplyOutlineToDropShadows=false
      bMiteredCorners=false
      bSeparateFillAlpha=false
      outlineMaterial=null
      outlineSize=0
      outlineColor=<rgba 0 0 0 1>
     }
    }
    renderTransform {angle=0, scale=<1 1>, shear=<0 0>, translation=<0 -2>}
   }
   slot {class=HorizontalBoxSlot, properties {verticalAlignment=VAlign_Center}}
  }
 ]
}
