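# DVR_Nav_BWD_2x_Cue is a Fortnite UI sound cue, likely the feedback for 2x backward navigation in replay playback.
# A single wave player serves as the root and plays the DVR_Nav_BWD_2x sound wave with no randomization or modulation.
# It routes through the non-critical in-game UI sound class.
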SoundCue /Game/Sounds/Fort_UI_Sounds/Replay/DVR_Nav_BWD_2x_Cue
root=WavePlayer_0
nodes [{id=WavePlayer_0, properties {soundWaveAssetPtr=&/DVR_Nav_BWD_2x}}]
properties {maxDistance=2097152, soundClassObject=&Fort_UI_InGame_NonCritical, volumeMultiplier=1}
