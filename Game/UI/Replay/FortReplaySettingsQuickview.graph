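# Replay HUD quickview panel that lists current replay camera settings such as shot, exposure, lens, focus, drone movement, tilt, lighting and motion blur.
# RefreshSettings runs on Construct, filling the value column from parent-class getters and the r.MotionBlur.Amount console variable.
# Construct also binds two delegates: toggling replay view settings collapses the panel, and spectator zoom changes update focal length while it is visible.
# PreConstruct evens out row padding in both columns.
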WidgetBlueprint /Game/UI/Replay/FortReplaySettingsQuickview
graphs [
 EventGraph {nodes [
   ForceCloseQuickAccess() {is=title K2Node_CustomEvent}
   > RemoveFromParent()
   n1 = ForceClose() {is=title K2Node_CustomEvent}
   > n2 = Switch on ESlateVisibility!(execute, Selection:ESlateVisibility Enum=GetVisibility!(self:Widget Object Reference) -> (ReturnValue:ESlateVisibility Enum=Visible)) -> (Visible:Exec, Collapsed:Exec, Hidden:Exec, HitTestInvisible:Exec, SelfHitTestInvisible:Exec) {is=title K2Node_SwitchEnum}
   SetVisibility(execute=[@n2.HitTestInvisible, @n2.SelfHitTestInvisible, @n2.Visible], InVisibility=Collapsed)
   n3 = UpdateZoom() {is=title K2Node_CustomEvent}
   > n4 = Switch on ESlateVisibility(Selection=GetVisibility() -> (ReturnValue=Visible))
   n5 = GetFocalLength!(self:Fort Replay Settings Quickview Object Reference) -> (ReturnValue:float)
   "SetText (Text)"(
    execute=[@n4.Visible, @n4.HitTestInvisible, @n4.SelfHitTestInvisible]
    self=Get Value_FocalLength() -> (Value_FocalLength:Common Text Object Reference)
    InText=Format Text!(
     Format="LOCTEXT(\"{value} {unit}\")"
     value:Text="To Text (Float)"(Value=@n5, bUseGrouping=false, MinimumIntegralDigits=1, MaximumIntegralDigits=10, MinimumFractionalDigits=0, MaximumFractionalDigits=0)
     unit:Text=Get FocalLengthUnit() -> (FocalLengthUnit:Text)
    ) -> (Result)
   )
   Event Construct()
   > RefreshSettings()
   > Bind Event to On Toggle Replay View Settings(
    self:Fort Replay Context Object Reference=Get FortReplayContext!() -> (ReturnValue:Fort Replay Context Object Reference) {is=title K2Node_GetBlueprintContextEx}
    Delegate=@n1
   ) {is=title K2Node_AddDelegate}
   > n6 = Cast To FortPlayerControllerSpectating!(execute, Object=GetOwningPlayer()) -> (then, CastFailed, AsFort Player Controller Spectating:Fort Player Controller Spectating Object Reference) {is=title K2Node_DynamicCast}
   > Bind Event to On Zoom Changed(self:Fort Player Controller Spectating Object Reference=@n6, Delegate=@n3) {is=title K2Node_AddDelegate}
   Event PreConstruct()
   > n7 = For Each Loop(Array:Array of Widget Object References=GetAllChildren(self=Get Titles() -> (Titles:Vertical Box Object Reference))) -> (Array Element:Widget Object Reference)
   > SetPadding!(
    execute
    self:Vertical Box Slot Object Reference=SlotAsVerticalBoxSlot(Widget=@n7)
    InPadding_Left:float=0
    InPadding_Top:float=0
    InPadding_Right:float=0
    InPadding_Bottom:float=15
   ) -> (then)
   > n8 = For Each Loop(Array=GetAllChildren(self=Get Values() -> (Values:Vertical Box Object Reference)))
   > SetPadding(self=SlotAsVerticalBoxSlot(Widget=@n8), InPadding_Left=0, InPadding_Top=0, InPadding_Right=0, InPadding_Bottom=15)
   > SetPadding(
    self=SlotAsVerticalBoxSlot(Widget=Get Text_MotionBlur() -> (Text_MotionBlur:Common Text Object Reference))
    InPadding_Left=0
    InPadding_Top=0
    InPadding_Right=0
    InPadding_Bottom=0
   )
   > SetPadding(
    self=SlotAsVerticalBoxSlot(Widget=Get Value_MotionBlur() -> (Value_MotionBlur:Common Text Object Reference))
    InPadding_Left=0
    InPadding_Top=0
    InPadding_Right=0
    InPadding_Bottom=0
   )
  ]}
 RefreshSettings {nodes [
   RefreshSettings() {is=title K2Node_FunctionEntry}
   > n1 = Branch(Condition="Equal (Integer)"(A=GetLoadedShotIndex!(self:Fort Replay Settings Quickview Object Reference) -> (ReturnValue:int=0), B=-1))
   > n2 = "SetText (Text)"(self=Get Value_Camera() -> (Value_Camera:Common Text Object Reference), InText="LOCTEXT(\"None\")")
   n3 = Branch(execute=[@n2.then, @n18.then], Condition=IsAutoExposureEnabled!(self:Fort Replay Settings Quickview Object Reference) -> (ReturnValue:bool))
   > n4 = "SetText (Text)"(self=Get Value_Exposure() -> (Value_Exposure:Common Text Object Reference), InText="LOCTEXT(\"On\")")
   n5 = GetFocalLength()
   "SetText (Text)"(
    execute=[@n4.then, @n17.then]
    self=Get Value_FocalLength()
    InText=Format Text(
     Format="LOCTEXT(\"{value} {unit}\")"
     value="To Text (Float)"(Value=@n5, bUseGrouping=false, MinimumIntegralDigits=1, MaximumIntegralDigits=10, MinimumFractionalDigits=0, MaximumFractionalDigits=0)
     unit=Get Loc_FocalLengthUnit() -> (Loc_FocalLengthUnit:Text)
    )
   )
   n6 = GetAperture!(self:Fort Replay Settings Quickview Object Reference) -> (ReturnValue:float)
   > "SetText (Text)"(
    self=Get Value_Aperture() -> (Value_Aperture:Common Text Object Reference)
    InText=Format Text!(
     Format="LOCTEXT(\"{unit}{value}\")"
     unit=Get Loc_ApertureUnit() -> (Loc_ApertureUnit:Text)
     value="To Text (Float)"(Value=@n6, bUseGrouping=false, MinimumIntegralDigits=1, MaximumIntegralDigits=10, MinimumFractionalDigits=1, MaximumFractionalDigits=1)
    ) -> (Result)
   )
   > n7 = Branch(Condition=IsAutoFocusEnabled!(self:Fort Replay Settings Quickview Object Reference) -> (ReturnValue:bool))
   > n8 = "SetText (Text)"(self=Get Value_AutoFocus() -> (Value_AutoFocus:Common Text Object Reference), InText="LOCTEXT(\"Auto\")")
   "SetText (Text)"(
    execute=[@n8.then, @n16.then]
    self=Get Value_LinearVelocity() -> (Value_LinearVelocity:Common Text Object Reference)
    InText="To Text (Float)"(
     Value=GetDroneDirectMovementLinearVelocity!(self:Fort Replay Settings Quickview Object Reference) -> (ReturnValue:float)
     bUseGrouping=false
     MinimumIntegralDigits=1
     MaximumIntegralDigits=10
     MinimumFractionalDigits=2
     MaximumFractionalDigits=2
    )
   )
   > "SetText (Text)"(
    self=Get Value_RotationalVelocity() -> (Value_RotationalVelocity:Common Text Object Reference)
    InText="To Text (Float)"(
     Value=GetDroneDirectMovementRotationVelocity!(self:Fort Replay Settings Quickview Object Reference) -> (ReturnValue:float)
     bUseGrouping=false
     MinimumIntegralDigits=1
     MaximumIntegralDigits=10
     MinimumFractionalDigits=2
     MaximumFractionalDigits=2
    )
   )
   n9 = GetDroneDutchAngle!(self:Fort Replay Settings Quickview Object Reference) -> (ReturnValue:float)
   > "SetText (Text)"(
    self=Get Value_DutchAngle() -> (Value_DutchAngle:Common Text Object Reference)
    InText=Format Text!(
     Format="LOCTEXT(\"{value}{unit}\")"
     value="To Text (Float)"(Value=@n9, bUseGrouping=false, MinimumIntegralDigits=1, MaximumIntegralDigits=3, MinimumFractionalDigits=0, MaximumFractionalDigits=0)
     unit=Get Loc_DegreeUnit() -> (Loc_DegreeUnit:Text)
    ) -> (Result)
   )
   > "SetText (Text)"(
    self=Get Value_DutchSpeed() -> (Value_DutchSpeed:Common Text Object Reference)
    InText="To Text (Float)"(
     Value=GetDroneDutchSpeed!(self:Fort Replay Settings Quickview Object Reference) -> (ReturnValue:float)
     bUseGrouping=false
     MinimumIntegralDigits=1
     MaximumIntegralDigits=10
     MinimumFractionalDigits=2
     MaximumFractionalDigits=2
    )
   )
   > "SetText (Text)"(
    self=Get Value_TimeOfDay() -> (Value_TimeOfDay:Common Text Object Reference)
    InText="To Text (Float)"(
     Value=GetCurrentTimeOfDay!(self:Fort Replay Settings Quickview Object Reference) -> (ReturnValue:float)
     bUseGrouping=false
     MinimumIntegralDigits=1
     MaximumIntegralDigits=2
     MinimumFractionalDigits=2
     MaximumFractionalDigits=2
    )
   )
   n10 = DirectionalLightAngle!(self:Fort Replay Settings Quickview Object Reference) -> (ReturnValue:float)
   > "SetText (Text)"(
    self=Get Value_DirectionalLightAngle() -> (Value_DirectionalLightAngle:Common Text Object Reference)
    InText=Format Text(
     Format="LOCTEXT(\"{value}{unit}\")"
     value="To Text (Float)"(Value=@n10, bUseGrouping=false, MinimumIntegralDigits=1, MaximumIntegralDigits=3, MinimumFractionalDigits=0, MaximumFractionalDigits=0)
     unit=Get Loc_DegreeUnit()
    )
   )
   > n11 = Branch(Condition=IsPausedTimeOfDay!(self:Fort Replay Settings Quickview Object Reference) -> (ReturnValue:bool))
   > n12 = "SetText (Text)"(self=Get Value_TimeOfDayPaused() -> (Value_TimeOfDayPaused:Common Text Object Reference), InText="LOCTEXT(\"On\")")
   n13 = GetConsoleVariableFloatValue!(execute=[@n12.then, @n14.then], VariableName:String=r.MotionBlur.Amount) -> (then, ReturnValue:float)
   > "SetText (Text)"(
    self=Get Value_MotionBlur()
    InText="To Text (Float)"(Value=@n13, bUseGrouping=false, MinimumIntegralDigits=1, MaximumIntegralDigits=2, MinimumFractionalDigits=1, MaximumFractionalDigits=2)
   )
   n14 = "SetText (Text)"(execute=@n11.else, self=Get Value_TimeOfDayPaused(), InText="LOCTEXT(\"Off\")")
   n15 = GetManualFocusDistance!(self:Fort Replay Settings Quickview Object Reference) -> (ReturnValue:float)
   n16 = "SetText (Text)"(
    execute=@n7.else
    self=Get Value_AutoFocus()
    InText=Format Text(
     Format="LOCTEXT(\"{value} {unit}\")"
     value="To Text (Float)"(Value=@n15, MinimumIntegralDigits=1, MaximumIntegralDigits=100, MinimumFractionalDigits=0, MaximumFractionalDigits=0)
     unit=Get Loc_FocusUnit() -> (Loc_FocusUnit:Text)
    )
   )
   n17 = "SetText (Text)"(
    execute=@n3.else
    self=Get Value_Exposure()
    InText="To Text (Float)"(
     Value=GetManualExposureBias!(self:Fort Replay Settings Quickview Object Reference) -> (ReturnValue:float)
     bUseGrouping=false
     MinimumIntegralDigits=1
     MaximumIntegralDigits=10
     MinimumFractionalDigits=2
     MaximumFractionalDigits=2
    )
   )
   n18 = "SetText (Text)"(
    execute=@n1.else
    self=Get Value_Camera()
    InText="To Text (Integer)"(Value=GetLoadedShotIndex() -> (ReturnValue=0), bUseGrouping=false, MaximumIntegralDigits=2)
   )
  ]}
]
variables [FocalLengthUnit {type=string, default=mm}]
widgets {
 info {inheritedWidgetCount=0, parentClass=/Script/FortniteUI.FortReplaySettingsQuickview, rootWidgetClass=&CanvasPanel, widgetCount=33}
 widgets [
  CanvasPanel_0 {widgetClassPath=&CanvasPanel, widget=CanvasPanel_0}
  SizeBox_0 {
   widgetClassPath=&SizeBox
   parent=CanvasPanel_0
   widget=SizeBox_0
   properties {
    bOverride_MaxDesiredHeight=true
    bOverride_MinDesiredHeight=true
    bOverride_WidthOverride=true
    heightOverride=500
    maxDesiredHeight=565
    minDesiredHeight=540
    widthOverride=375
   }
   slot {
    class=CanvasPanelSlot
    properties {bAutoSize=true, layoutData {alignment=<1 0>, anchors {maximum=<1 0>, minimum=<1 0>}, offsets=<ltrb -40 40 100 30>}}
   }
  }
  Overlay_129 {widgetClassPath=&Overlay, parent=SizeBox_0, widget=Overlay_129, slot {class=SizeBoxSlot}}
  WBP_UIKit_Backplate {
   widgetClassPath=&WBP_UIKit_Backplate
   bIsVariable=true
   parent=Overlay_129
   widget=WBP_UIKit_Backplate
   properties {backplateMaterial=/FortUILibrary/Systems/Backplate/MaterialInstances/MI_UIKit_Backplate_Fallback, brightness=Dark}
   slot {class=OverlaySlot, properties {horizontalAlignment=HAlign_Fill, verticalAlignment=VAlign_Fill}}
  }
  Columns {
   widgetClassPath=&HorizontalBox
   parent=Overlay_129
   widget=Columns
   slot {class=OverlaySlot, properties {horizontalAlignment=HAlign_Fill, verticalAlignment=VAlign_Fill, padding=<ltrb 18 15 18 15>}}
  }
  Titles {
   widgetClassPath=&VerticalBox
   bIsVariable=true
   parent=Columns
   widget=Titles
   slot {class=HorizontalBoxSlot, properties {horizontalAlignment=HAlign_Left}}
  }
  Text_Camera {
   widgetClassPath=&CommonTextBlock
   parent=Titles
   widget=Text_Camera
   properties {
    text=Shot:
    font {
     bForceMonospaced=false
     bMaterialIsStencil=false
     fontMaterial=null
     fontObject=&HeadingNow
     letterSpacing=0
     monospacedWidth=1
     size=15.75
     skewAmount=0
     typefaceFontName=74 Regular
     outlineSettings {
      bApplyOutlineToDropShadows=false
      bMiteredCorners=false
      bSeparateFillAlpha=false
      outlineMaterial=null
      outlineSize=0
      outlineColor=<rgba 0 0 0 1>
     }
    }
    shadowOffset=<0 0>
   }
   slot {class=VerticalBoxSlot, properties {horizontalAlignment=HAlign_Left, verticalAlignment=VAlign_Top, padding=<ltrb 0 0 0 5>}}
  }
  Text_Exposure {
   widgetClassPath=&CommonTextBlock
   parent=Titles
   widget=Text_Exposure
   properties {
    text=Exposure:
    font {
     bForceMonospaced=false
     bMaterialIsStencil=false
     fontMaterial=null
     fontObject=&HeadingNow
     letterSpacing=0
     monospacedWidth=1
     size=15.75
     skewAmount=0
     typefaceFontName=74 Regular
     outlineSettings {
      bApplyOutlineToDropShadows=false
      bMiteredCorners=false
      bSeparateFillAlpha=false
      outlineMaterial=null
      outlineSize=0
      outlineColor=<rgba 0 0 0 1>
     }
    }
    shadowOffset=<0 0>
   }
   slot {class=VerticalBoxSlot, properties {horizontalAlignment=HAlign_Left, verticalAlignment=VAlign_Top, padding=<ltrb 0 5 0 5>}}
  }
  Text_FocalLength {
   widgetClassPath=&CommonTextBlock
   parent=Titles
   widget=Text_FocalLength
   properties {
    text=Focal Length:
    font {
     bForceMonospaced=false
     bMaterialIsStencil=false
     fontMaterial=null
     fontObject=&HeadingNow
     letterSpacing=0
     monospacedWidth=1
     size=15.75
     skewAmount=0
     typefaceFontName=74 Regular
     outlineSettings {
      bApplyOutlineToDropShadows=false
      bMiteredCorners=false
      bSeparateFillAlpha=false
      outlineMaterial=null
      outlineSize=0
      outlineColor=<rgba 0 0 0 1>
     }
    }
    shadowOffset=<0 0>
   }
   slot {class=VerticalBoxSlot, properties {horizontalAlignment=HAlign_Left, verticalAlignment=VAlign_Top, padding=<ltrb 0 5 0 5>}}
  }
  Text_Aperture {
   widgetClassPath=&CommonTextBlock
   parent=Titles
   widget=Text_Aperture
   properties {
    text=Aperture:
    font {
     bForceMonospaced=false
     bMaterialIsStencil=false
     fontMaterial=null
     fontObject=&HeadingNow
     letterSpacing=0
     monospacedWidth=1
     size=15.75
     skewAmount=0
     typefaceFontName=74 Regular
     outlineSettings {
      bApplyOutlineToDropShadows=false
      bMiteredCorners=false
      bSeparateFillAlpha=false
      outlineMaterial=null
      outlineSize=0
      outlineColor=<rgba 0 0 0 1>
     }
    }
    shadowOffset=<0 0>
   }
   slot {class=VerticalBoxSlot, properties {horizontalAlignment=HAlign_Left, verticalAlignment=VAlign_Top, padding=<ltrb 0 5 0 5>}}
  }
  Text_AutoFocus {
   widgetClassPath=&CommonTextBlock
   parent=Titles
   widget=Text_AutoFocus
   properties {
    text=Focus:
    font {
     bForceMonospaced=false
     bMaterialIsStencil=false
     fontMaterial=null
     fontObject=&HeadingNow
     letterSpacing=0
     monospacedWidth=1
     size=15.75
     skewAmount=0
     typefaceFontName=74 Regular
     outlineSettings {
      bApplyOutlineToDropShadows=false
      bMiteredCorners=false
      bSeparateFillAlpha=false
      outlineMaterial=null
      outlineSize=0
      outlineColor=<rgba 0 0 0 1>
     }
    }
    shadowOffset=<0 0>
   }
   slot {class=VerticalBoxSlot, properties {horizontalAlignment=HAlign_Left, verticalAlignment=VAlign_Top, padding=<ltrb 0 5 0 5>}}
  }
  Text_LinearVelocity {
   widgetClassPath=&CommonTextBlock
   parent=Titles
   widget=Text_LinearVelocity
   properties {
    text=Linear Velocity:
    font {
     bForceMonospaced=false
     bMaterialIsStencil=false
     fontMaterial=null
     fontObject=&HeadingNow
     letterSpacing=0
     monospacedWidth=1
     size=15.75
     skewAmount=0
     typefaceFontName=74 Regular
     outlineSettings {
      bApplyOutlineToDropShadows=false
      bMiteredCorners=false
      bSeparateFillAlpha=false
      outlineMaterial=null
      outlineSize=0
      outlineColor=<rgba 0 0 0 1>
     }
    }
    shadowOffset=<0 0>
   }
   slot {class=VerticalBoxSlot, properties {horizontalAlignment=HAlign_Left, verticalAlignment=VAlign_Top, padding=<ltrb 0 5 0 5>}}
  }
  Text_RotationalVelocity {
   widgetClassPath=&CommonTextBlock
   parent=Titles
   widget=Text_RotationalVelocity
   properties {
    text=Rotational Velocity:
    font {
     bForceMonospaced=false
     bMaterialIsStencil=false
     fontMaterial=null
     fontObject=&HeadingNow
     letterSpacing=0
     monospacedWidth=1
     size=15.75
     skewAmount=0
     typefaceFontName=74 Regular
     outlineSettings {
      bApplyOutlineToDropShadows=false
      bMiteredCorners=false
      bSeparateFillAlpha=false
      outlineMaterial=null
      outlineSize=0
      outlineColor=<rgba 0 0 0 1>
     }
    }
    shadowOffset=<0 0>
   }
   slot {class=VerticalBoxSlot, properties {horizontalAlignment=HAlign_Left, verticalAlignment=VAlign_Top, padding=<ltrb 0 5 0 5>}}
  }
  Text_DutchAngle {
   widgetClassPath=&CommonTextBlock
   parent=Titles
   widget=Text_DutchAngle
   properties {
    text=Tilt:
    font {
     bForceMonospaced=false
     bMaterialIsStencil=false
     fontMaterial=null
     fontObject=&HeadingNow
     letterSpacing=0
     monospacedWidth=1
     size=15.75
     skewAmount=0
     typefaceFontName=74 Regular
     outlineSettings {
      bApplyOutlineToDropShadows=false
      bMiteredCorners=false
      bSeparateFillAlpha=false
      outlineMaterial=null
      outlineSize=0
      outlineColor=<rgba 0 0 0 1>
     }
    }
    shadowOffset=<0 0>
   }
   slot {class=VerticalBoxSlot, properties {horizontalAlignment=HAlign_Left, verticalAlignment=VAlign_Top, padding=<ltrb 0 5 0 5>}}
  }
  Text_DutchSpeed {
   widgetClassPath=&CommonTextBlock
   parent=Titles
   widget=Text_DutchSpeed
   properties {
    text=Tilt Angle:
    font {
     bForceMonospaced=false
     bMaterialIsStencil=false
     fontMaterial=null
     fontObject=&HeadingNow
     letterSpacing=0
     monospacedWidth=1
     size=15.75
     skewAmount=0
     typefaceFontName=74 Regular
     outlineSettings {
      bApplyOutlineToDropShadows=false
      bMiteredCorners=false
      bSeparateFillAlpha=false
      outlineMaterial=null
      outlineSize=0
      outlineColor=<rgba 0 0 0 1>
     }
    }
    shadowOffset=<0 0>
   }
   slot {class=VerticalBoxSlot, properties {horizontalAlignment=HAlign_Left, verticalAlignment=VAlign_Top, padding=<ltrb 0 5 0 5>}}
  }
  Text_TimeOfDay {
   widgetClassPath=&CommonTextBlock
   parent=Titles
   widget=Text_TimeOfDay
   properties {
    text=Time Of Day:
    font {
     bForceMonospaced=false
     bMaterialIsStencil=false
     fontMaterial=null
     fontObject=&HeadingNow
     letterSpacing=0
     monospacedWidth=1
     size=15.75
     skewAmount=0
     typefaceFontName=74 Regular
     outlineSettings {
      bApplyOutlineToDropShadows=false
      bMiteredCorners=false
      bSeparateFillAlpha=false
      outlineMaterial=null
      outlineSize=0
      outlineColor=<rgba 0 0 0 1>
     }
    }
    shadowOffset=<0 0>
   }
   slot {class=VerticalBoxSlot, properties {horizontalAlignment=HAlign_Left, verticalAlignment=VAlign_Top, padding=<ltrb 0 5 0 5>}}
  }
  Text_TimeOfDayPaused {
   widgetClassPath=&CommonTextBlock
   parent=Titles
   widget=Text_TimeOfDayPaused
   properties {
    text=TOD Paused:
    font {
     bForceMonospaced=false
     bMaterialIsStencil=false
     fontMaterial=null
     fontObject=&HeadingNow
     letterSpacing=0
     monospacedWidth=1
     size=15.75
     skewAmount=0
     typefaceFontName=74 Regular
     outlineSettings {
      bApplyOutlineToDropShadows=false
      bMiteredCorners=false
      bSeparateFillAlpha=false
      outlineMaterial=null
      outlineSize=0
      outlineColor=<rgba 0 0 0 1>
     }
    }
    shadowOffset=<0 0>
   }
   slot {class=VerticalBoxSlot, properties {horizontalAlignment=HAlign_Left, verticalAlignment=VAlign_Top, padding=<ltrb 0 5 0 5>}}
  }
  Text_DirectionalLightAngle {
   widgetClassPath=&CommonTextBlock
   parent=Titles
   widget=Text_DirectionalLightAngle
   properties {
    text=Light Angle:
    font {
     bForceMonospaced=false
     bMaterialIsStencil=false
     fontMaterial=null
     fontObject=&HeadingNow
     letterSpacing=0
     monospacedWidth=1
     size=15.75
     skewAmount=0
     typefaceFontName=74 Regular
     outlineSettings {
      bApplyOutlineToDropShadows=false
      bMiteredCorners=false
      bSeparateFillAlpha=false
      outlineMaterial=null
      outlineSize=0
      outlineColor=<rgba 0 0 0 1>
     }
    }
    shadowOffset=<0 0>
   }
   slot {class=VerticalBoxSlot, properties {horizontalAlignment=HAlign_Left, verticalAlignment=VAlign_Top, padding=<ltrb 0 0 0 15>}}
  }
  Text_MotionBlur {
   widgetClassPath=&CommonTextBlock
   bIsVariable=true
   parent=Titles
   widget=Text_MotionBlur
   properties {
    text=Motion Blur:
    font {
     bForceMonospaced=false
     bMaterialIsStencil=false
     fontMaterial=null
     fontObject=&HeadingNow
     letterSpacing=0
     monospacedWidth=1
     size=15.75
     skewAmount=0
     typefaceFontName=74 Regular
     outlineSettings {
      bApplyOutlineToDropShadows=false
      bMiteredCorners=false
      bSeparateFillAlpha=false
      outlineMaterial=null
      outlineSize=0
      outlineColor=<rgba 0 0 0 1>
     }
    }
    shadowOffset=<0 0>
   }
   slot {class=VerticalBoxSlot, properties {horizontalAlignment=HAlign_Left, verticalAlignment=VAlign_Top}}
  }
  Values {
   widgetClassPath=&VerticalBox
   bIsVariable=true
   parent=Columns
   widget=Values
   slot {class=HorizontalBoxSlot, properties {horizontalAlignment=HAlign_Right, size {sizeRule=Fill, value=1}}}
  }
  Value_Camera {
   widgetClassPath=&CommonTextBlock
   bIsVariable=true
   parent=Values
   widget=Value_Camera
   properties {
    text=0
    font {
     bForceMonospaced=false
     bMaterialIsStencil=false
     fontMaterial=null
     fontObject=&HeadingNow
     letterSpacing=0
     monospacedWidth=1
     size=15.75
     skewAmount=0
     typefaceFontName=75 Medium
     outlineSettings {
      bApplyOutlineToDropShadows=false
      bMiteredCorners=false
      bSeparateFillAlpha=false
      outlineMaterial=null
      outlineSize=0
      outlineColor=<rgba 0 0 0 1>
     }
    }
    shadowOffset=<0 0>
   }
   slot {class=VerticalBoxSlot, properties {horizontalAlignment=HAlign_Right, verticalAlignment=VAlign_Top, padding=<ltrb 0 0 0 5>}}
  }
  Value_Exposure {
   widgetClassPath=&CommonTextBlock
   bIsVariable=true
   parent=Values
   widget=Value_Exposure
   properties {
    text=0.00
    font {
     bForceMonospaced=false
     bMaterialIsStencil=false
     fontMaterial=null
     fontObject=&HeadingNow
     letterSpacing=0
     monospacedWidth=1
     size=15.75
     skewAmount=0
     typefaceFontName=75 Medium
     outlineSettings {
      bApplyOutlineToDropShadows=false
      bMiteredCorners=false
      bSeparateFillAlpha=false
      outlineMaterial=null
      outlineSize=0
      outlineColor=<rgba 0 0 0 1>
     }
    }
    shadowOffset=<0 0>
   }
   slot {class=VerticalBoxSlot, properties {horizontalAlignment=HAlign_Right, verticalAlignment=VAlign_Top, padding=<ltrb 0 5 0 5>}}
  }
  Value_FocalLength {
   widgetClassPath=&CommonTextBlock
   bIsVariable=true
   parent=Values
   widget=Value_FocalLength
   properties {
    text=35 mm
    font {
     bForceMonospaced=false
     bMaterialIsStencil=false
     fontMaterial=null
     fontObject=&HeadingNow
     letterSpacing=0
     monospacedWidth=1
     size=15.75
     skewAmount=0
     typefaceFontName=75 Medium
     outlineSettings {
      bApplyOutlineToDropShadows=false
      bMiteredCorners=false
      bSeparateFillAlpha=false
      outlineMaterial=null
      outlineSize=0
      outlineColor=<rgba 0 0 0 1>
     }
    }
    shadowOffset=<0 0>
   }
   slot {class=VerticalBoxSlot, properties {horizontalAlignment=HAlign_Right, verticalAlignment=VAlign_Top, padding=<ltrb 0 5 0 5>}}
  }
  Value_Aperture {
   widgetClassPath=&CommonTextBlock
   bIsVariable=true
   parent=Values
   widget=Value_Aperture
   properties {
    text=f/1.0
    font {
     bForceMonospaced=false
     bMaterialIsStencil=false
     fontMaterial=null
     fontObject=&HeadingNow
     letterSpacing=0
     monospacedWidth=1
     size=15.75
     skewAmount=0
     typefaceFontName=75 Medium
     outlineSettings {
      bApplyOutlineToDropShadows=false
      bMiteredCorners=false
      bSeparateFillAlpha=false
      outlineMaterial=null
      outlineSize=0
      outlineColor=<rgba 0 0 0 1>
     }
    }
    shadowOffset=<0 0>
   }
   slot {class=VerticalBoxSlot, properties {horizontalAlignment=HAlign_Right, verticalAlignment=VAlign_Top, padding=<ltrb 0 5 0 5>}}
  }
  Value_AutoFocus {
   widgetClassPath=&CommonTextBlock
   bIsVariable=true
   parent=Values
   widget=Value_AutoFocus
   properties {
    text=Auto
    font {
     bForceMonospaced=false
     bMaterialIsStencil=false
     fontMaterial=null
     fontObject=&HeadingNow
     letterSpacing=0
     monospacedWidth=1
     size=15.75
     skewAmount=0
     typefaceFontName=75 Medium
     outlineSettings {
      bApplyOutlineToDropShadows=false
      bMiteredCorners=false
      bSeparateFillAlpha=false
      outlineMaterial=null
      outlineSize=0
      outlineColor=<rgba 0 0 0 1>
     }
    }
    shadowOffset=<0 0>
   }
   slot {class=VerticalBoxSlot, properties {horizontalAlignment=HAlign_Right, verticalAlignment=VAlign_Top, padding=<ltrb 0 5 0 5>}}
  }
  Value_LinearVelocity {
   widgetClassPath=&CommonTextBlock
   bIsVariable=true
   parent=Values
   widget=Value_LinearVelocity
   properties {
    text=0.00
    font {
     bForceMonospaced=false
     bMaterialIsStencil=false
     fontMaterial=null
     fontObject=&HeadingNow
     letterSpacing=0
     monospacedWidth=1
     size=15.75
     skewAmount=0
     typefaceFontName=75 Medium
     outlineSettings {
      bApplyOutlineToDropShadows=false
      bMiteredCorners=false
      bSeparateFillAlpha=false
      outlineMaterial=null
      outlineSize=0
      outlineColor=<rgba 0 0 0 1>
     }
    }
    shadowOffset=<0 0>
   }
   slot {class=VerticalBoxSlot, properties {horizontalAlignment=HAlign_Right, verticalAlignment=VAlign_Top, padding=<ltrb 0 5 0 5>}}
  }
  Value_RotationalVelocity {
   widgetClassPath=&CommonTextBlock
   bIsVariable=true
   parent=Values
   widget=Value_RotationalVelocity
   properties {
    text=0.00
    font {
     bForceMonospaced=false
     bMaterialIsStencil=false
     fontMaterial=null
     fontObject=&HeadingNow
     letterSpacing=0
     monospacedWidth=1
     size=15.75
     skewAmount=0
     typefaceFontName=75 Medium
     outlineSettings {
      bApplyOutlineToDropShadows=false
      bMiteredCorners=false
      bSeparateFillAlpha=false
      outlineMaterial=null
      outlineSize=0
      outlineColor=<rgba 0 0 0 1>
     }
    }
    shadowOffset=<0 0>
   }
   slot {class=VerticalBoxSlot, properties {horizontalAlignment=HAlign_Right, verticalAlignment=VAlign_Top, padding=<ltrb 0 5 0 5>}}
  }
  Value_DutchAngle {
   widgetClassPath=&CommonTextBlock
   bIsVariable=true
   parent=Values
   widget=Value_DutchAngle
   properties {
    text="0°"
    font {
     bForceMonospaced=false
     bMaterialIsStencil=false
     fontMaterial=null
     fontObject=&HeadingNow
     letterSpacing=0
     monospacedWidth=1
     size=15.75
     skewAmount=0
     typefaceFontName=75 Medium
     outlineSettings {
      bApplyOutlineToDropShadows=false
      bMiteredCorners=false
      bSeparateFillAlpha=false
      outlineMaterial=null
      outlineSize=0
      outlineColor=<rgba 0 0 0 1>
     }
    }
    shadowOffset=<0 0>
   }
   slot {class=VerticalBoxSlot, properties {horizontalAlignment=HAlign_Right, verticalAlignment=VAlign_Top, padding=<ltrb 0 5 0 5>}}
  }
  Value_DutchSpeed {
   widgetClassPath=&CommonTextBlock
   bIsVariable=true
   parent=Values
   widget=Value_DutchSpeed
   properties {
    text=0.00
    font {
     bForceMonospaced=false
     bMaterialIsStencil=false
     fontMaterial=null
     fontObject=&HeadingNow
     letterSpacing=0
     monospacedWidth=1
     size=15.75
     skewAmount=0
     typefaceFontName=75 Medium
     outlineSettings {
      bApplyOutlineToDropShadows=false
      bMiteredCorners=false
      bSeparateFillAlpha=false
      outlineMaterial=null
      outlineSize=0
      outlineColor=<rgba 0 0 0 1>
     }
    }
    shadowOffset=<0 0>
   }
   slot {class=VerticalBoxSlot, properties {horizontalAlignment=HAlign_Right, verticalAlignment=VAlign_Top, padding=<ltrb 0 5 0 5>}}
  }
  Value_TimeOfDay {
   widgetClassPath=&CommonTextBlock
   bIsVariable=true
   parent=Values
   widget=Value_TimeOfDay
   properties {
    text=12.00
    font {
     bForceMonospaced=false
     bMaterialIsStencil=false
     fontMaterial=null
     fontObject=&HeadingNow
     letterSpacing=0
     monospacedWidth=1
     size=15.75
     skewAmount=0
     typefaceFontName=75 Medium
     outlineSettings {
      bApplyOutlineToDropShadows=false
      bMiteredCorners=false
      bSeparateFillAlpha=false
      outlineMaterial=null
      outlineSize=0
      outlineColor=<rgba 0 0 0 1>
     }
    }
    shadowOffset=<0 0>
   }
   slot {class=VerticalBoxSlot, properties {horizontalAlignment=HAlign_Right, verticalAlignment=VAlign_Top, padding=<ltrb 0 5 0 5>}}
  }
  Value_TimeOfDayPaused {
   widgetClassPath=&CommonTextBlock
   bIsVariable=true
   parent=Values
   widget=Value_TimeOfDayPaused
   properties {
    text=Off
    font {
     bForceMonospaced=false
     bMaterialIsStencil=false
     fontMaterial=null
     fontObject=&HeadingNow
     letterSpacing=0
     monospacedWidth=1
     size=15.75
     skewAmount=0
     typefaceFontName=75 Medium
     outlineSettings {
      bApplyOutlineToDropShadows=false
      bMiteredCorners=false
      bSeparateFillAlpha=false
      outlineMaterial=null
      outlineSize=0
      outlineColor=<rgba 0 0 0 1>
     }
    }
    shadowOffset=<0 0>
   }
   slot {class=VerticalBoxSlot, properties {horizontalAlignment=HAlign_Right, verticalAlignment=VAlign_Top, padding=<ltrb 0 5 0 5>}}
  }
  Value_DirectionalLightAngle {
   widgetClassPath=&CommonTextBlock
   bIsVariable=true
   parent=Values
   widget=Value_DirectionalLightAngle
   properties {
    text="135°"
    font {
     bForceMonospaced=false
     bMaterialIsStencil=false
     fontMaterial=null
     fontObject=&HeadingNow
     letterSpacing=0
     monospacedWidth=1
     size=15.75
     skewAmount=0
     typefaceFontName=75 Medium
     outlineSettings {
      bApplyOutlineToDropShadows=false
      bMiteredCorners=false
      bSeparateFillAlpha=false
      outlineMaterial=null
      outlineSize=0
      outlineColor=<rgba 0 0 0 1>
     }
    }
    shadowOffset=<0 0>
   }
   slot {class=VerticalBoxSlot, properties {horizontalAlignment=HAlign_Right, verticalAlignment=VAlign_Top, padding=<ltrb 0 0 0 15>}}
  }
  Value_MotionBlur {
   widgetClassPath=&CommonTextBlock
   bIsVariable=true
   parent=Values
   widget=Value_MotionBlur
   properties {
    text=0.5
    font {
     bForceMonospaced=false
     bMaterialIsStencil=false
     fontMaterial=null
     fontObject=&HeadingNow
     letterSpacing=0
     monospacedWidth=1
     size=15.75
     skewAmount=0
     typefaceFontName=75 Medium
     outlineSettings {
      bApplyOutlineToDropShadows=false
      bMiteredCorners=false
      bSeparateFillAlpha=false
      outlineMaterial=null
      outlineSize=0
      outlineColor=<rgba 0 0 0 1>
     }
    }
    shadowOffset=<0 0>
   }
   slot {class=VerticalBoxSlot, properties {horizontalAlignment=HAlign_Right, verticalAlignment=VAlign_Top}}
  }
 ]
}
