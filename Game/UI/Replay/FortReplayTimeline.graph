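# Replay timeline bar built on FortTimelineBase, showing current and end times around a progress bar, with in/out points for looped playback.
# Construct enables looping; in non-shipping builds, I and O set interval points, Backspace clears them and L toggles looping through FortReplayContext.
# Indicators sit at time divided by replay end time, times canvas width.
# Tick retries caching that width until nonzero; camera-type changes re-measure it and reposition both indicators.
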WidgetBlueprint /Game/UI/Replay/FortReplayTimeline
graphs [
 EventGraph {nodes [
   n1 = RetryGetTimelineSize!() -> (OutputDelegate, then, SpectatorPC:Fort Player Controller Spectating Object Reference, NewCameraType:ESpectatorCameraType Enum=ThirdPerson) {is=title K2Node_CustomEvent}
   n2 = Get CanvasPanelIndicators() -> (CanvasPanelIndicators:Canvas Panel Object Reference)
   > Set CanvasPanelSize(CanvasPanelSize:double=GetLocalSize!(Geometry:"Geometry Structure (by ref)"=GetCachedGeometry!(self:Widget Object Reference=@n2) -> (ReturnValue:"Geometry Structure (by ref)")) -> (ReturnValue_X:double, ReturnValue_Y:double)) -> (Output_Get:double)
   > ForceMoveEndPoint()
   > ForceMoveStartPoint()
   ForceMoveStartPoint() {is=title K2Node_CustomEvent}
   > n3 = Cast To CanvasPanelSlot!(
    execute
    Object=Get Slot!(self:Widget Object Reference=Get In_Indicator() -> (In_Indicator:Image Object Reference)) -> (Slot:Panel Slot Object Reference)
   ) -> (then, CastFailed, AsCanvas Panel Slot:Canvas Panel Slot Object Reference) {is=title K2Node_DynamicCast}
   n4 = GetReplayEndTime!() -> (EndTime:float)
   n5 = Get InTime() -> (InTime:double)
   > SetPosition!(
    execute
    self:Canvas Panel Slot Object Reference=@n3
    InPosition_X:double=float * float(A=float / float(A=@n5, B=@n4), B=Get CanvasPanelSize() -> (CanvasPanelSize:double)) {is=title K2Node_PromotableOperator}
    InPosition_Y:double=0
   ) -> (then)
   ForceMoveEndPoint() {is=title K2Node_CustomEvent}
   > n6 = Cast To CanvasPanelSlot(Object=Get Slot(self=Get Out_Indicator() -> (Out_Indicator:Image Object Reference)))
   n7 = GetReplayEndTime()
   n8 = Get OutTime() -> (OutTime:double)
   > SetPosition(self=@n6, InPosition_X=float * float(A=float / float(A=@n8, B=@n7), B=Get CanvasPanelSize()), InPosition_Y=0)
   Event PreConstruct()
   n9 = Get ProgressBar_TimeElapsed() -> (ProgressBar_TimeElapsed:Progress Bar Object Reference)
   n10 = GetDynamicMaterial!(Brush:"Slate Brush Structure (by ref)"=Break Progress Bar Style!(ProgressBarStyle:"Progress Bar Style Structure (by ref)"=Get WidgetStyle!(self:Progress Bar Object Reference=@n9) -> (WidgetStyle:Progress Bar Style Structure)) -> (BackgroundImage:Slate Brush Structure, FillImage:Slate Brush Structure, MarqueeImage:Slate Brush Structure, EnableFillAnimation:bool=~) {is=title K2Node_BreakStruct}.FillImage) -> (ReturnValue)
   > SetVectorParameterValue(self=@n10, ParameterName=ColorB, Value=(R=0.8802,G=0.8802,B=0.8802,A=1))
   > SetVectorParameterValue(self=@n10, ParameterName=ColorA, Value=(R=0.1458,G=0.1458,B=0.1458,A=1))
   Event Construct()
   > Set InTime(InTime:double=GetReplayStartTime!() -> (StartTime:float)) -> (Output_Get:double)
   > Set OutTime(OutTime:double=GetReplayEndTime()) -> (Output_Get:double)
   > SetPlayIntervalLoop!(
    execute
    self:Fort Replay Context Object Reference=Get FortReplayContext!() -> (ReturnValue:Fort Replay Context Object Reference) {is=title K2Node_GetBlueprintContextEx}
    bLoop:bool=true
   ) -> (then)
   > SetLoopPauseTime!(execute, self:Fort Replay Context Object Reference=Get FortReplayContext(), InBufferTime:float=0) -> (then)
   > n11 = Cast To FortPlayerControllerSpectating!(execute, Object=GetPlayerController!(PlayerIndex:int=0) -> (ReturnValue:Player Controller Object Reference)) -> (then, CastFailed, AsFort Player Controller Spectating:Fort Player Controller Spectating Object Reference) {is=title K2Node_DynamicCast}
   > Bind Event to On Camera Type Changed(self:Fort Player Controller Spectating Object Reference=@n11, Delegate=@n1) {is=title K2Node_AddDelegate}
   Event Tick()
   > n12 = Branch(Condition=Get bIsSizeSet() -> (bIsSizeSet:bool))
   n13 = Get CanvasPanelIndicators()
   n14 = Set CanvasPanelSize(execute=@n12.else, CanvasPanelSize=GetLocalSize(Geometry=GetCachedGeometry(self=@n13)))
   > n15 = Branch(Condition="Equal (Float)"(A=@n14))
   Set bIsSizeSet(execute=@n15.else, bIsSizeSet:bool=true) -> (Output_Get:bool)
   L!() -> (Pressed:Exec, Released:Exec, Key:Key Structure) {is=title K2Node_InputKey, inputKey=L}
   > n16 = Branch(Condition=IsShippingBuild!(self:Fort Runtime Options Object Reference=GetRuntimeOptions()) -> (ReturnValue:bool))
   n17 = Flip Flop!(None:Exec=@n16.else) -> (A:Exec, B:Exec, IsA:bool=~) {is=title K2Node_MacroInstance}
   > SetPlayIntervalLoop(self=Get FortReplayContext(), bLoop=false)
   > ShowMessage!(execute, self, Notification:Text="LOCTEXT(\"Looping Disabled\")") -> (then)
   SetPlayIntervalLoop(execute=@n17.B, self=Get FortReplayContext(), bLoop=true)
   > ShowMessage(Notification="LOCTEXT(\"Looping Enabled\")")
   I!() -> (Pressed:Exec, Released:Exec, Key:Key Structure) {is=title K2Node_InputKey, inputKey=I}
   > n18 = Branch(Condition=IsShippingBuild(self=GetRuntimeOptions()))
   n19 = SetPlayInterval!(
    execute=@n18.else
    self:Fort Replay Context Object Reference=Get FortReplayContext()
    StartTime:float=GetCurrentReplayTime!() -> (TimeNow:float)
    EndTime:float=Get OutTime()
   ) -> (then, ReturnValue:bool)
   > Branch(Condition=@n19)
   > Set InTime(InTime=GetCurrentReplayTime())
   > SetVisibility(self=Get In_Indicator(), InVisibility=Visible)
   > n20 = Cast To CanvasPanelSlot(Object=Get Slot(self=Get In_Indicator()))
   n21 = GetCurrentReplayTime()
   n22 = GetReplayEndTime()
   > SetPosition(self=@n20, InPosition_X=float * float(A=float / float(A=@n21, B=@n22), B=Get CanvasPanelSize()), InPosition_Y=0)
   O!() -> (Pressed:Exec, Released:Exec, Key:Key Structure) {is=title K2Node_InputKey, inputKey=O}
   > n23 = Branch(Condition=IsShippingBuild(self=GetRuntimeOptions()))
   n24 = SetPlayInterval(execute=@n23.else, self=Get FortReplayContext(), StartTime=Get InTime(), EndTime=GetCurrentReplayTime())
   > Branch(Condition=@n24)
   > Set OutTime(OutTime=GetCurrentReplayTime())
   > SetVisibility(self=Get Out_Indicator(), InVisibility=Visible)
   > n25 = Cast To CanvasPanelSlot(Object=Get Slot(self=Get Out_Indicator()))
   n26 = GetCurrentReplayTime()
   n27 = GetReplayEndTime()
   > SetPosition(self=@n25, InPosition_X=float * float(A=float / float(A=@n26, B=@n27), B=Get CanvasPanelSize()), InPosition_Y=0)
   Delete!() -> (Pressed:Exec, Released:Exec, Key:Key Structure) {is=title K2Node_InputKey, inputKey=BackSpace}
   > n28 = Branch(Condition=IsShippingBuild(self=GetRuntimeOptions()))
   ClearPlayInterval(execute=@n28.else, self:Fort Replay Context Object Reference=Get FortReplayContext())
   > SetVisibility(self=[Get In_Indicator(), Get Out_Indicator()], InVisibility=Collapsed)
   > Set InTime(InTime=GetReplayStartTime())
   > Set OutTime(OutTime=GetReplayEndTime())
   > ShowMessage(Notification="LOCTEXT(\"In and Out Points Cleared\")")
  ]}
 GetCurrentReplayTime {nodes [
   GetCurrentReplayTime() {is=title K2Node_FunctionEntry}
   > Return Node!(
    execute
    TimeNow:float=Get!(self:Fort Replay Context Object Reference=Get FortReplayContext()) -> (
     PlaybackState_StartTime:float=~
     PlaybackState_EndTime:float=~
     PlaybackState_TimeNow:float=~
     PlaybackState_bIsPaused:bool=~
     PlaybackState_PlaybackSpeedMultiplier:float=~
     PlaybackState_HUDVisibility:EHudVisibilityState Enum
     PlaybackState_bLevelStreaming:bool=~
     PlaybackState_bHasRelevancyZone:bool=~
     PlaybackState_bHasPlayInterval:bool=~
     PlaybackState_PlayIntervalStartTime:float=~
     PlaybackState_PlayIntervalEndTime:float=~
     PlaybackState_bLoopInterval:bool=~
     PlaybackState_IntervalLoopBufferTime:float=~
    ).PlaybackState_TimeNow
   )
  ]}
 GetReplayEndTime {nodes [
   GetReplayEndTime() {is=title K2Node_FunctionEntry}
   > Return Node!(
    execute
    EndTime:float=Get(self=Get FortReplayContext()) -> (
     PlaybackState_StartTime=~
     PlaybackState_EndTime=~
     PlaybackState_TimeNow=~
     PlaybackState_bIsPaused=~
     PlaybackState_PlaybackSpeedMultiplier=~
     PlaybackState_bLevelStreaming=~
     PlaybackState_bHasRelevancyZone=~
     PlaybackState_bHasPlayInterval=~
     PlaybackState_PlayIntervalStartTime=~
     PlaybackState_PlayIntervalEndTime=~
     PlaybackState_bLoopInterval=~
     PlaybackState_IntervalLoopBufferTime=~
    ).PlaybackState_EndTime
   )
  ]}
 GetReplayStartTime {nodes [
   GetReplayStartTime() {is=title K2Node_FunctionEntry}
   > Return Node!(
    execute
    StartTime:float=Get(self=Get FortReplayContext()) -> (
     PlaybackState_StartTime=~
     PlaybackState_EndTime=~
     PlaybackState_TimeNow=~
     PlaybackState_bIsPaused=~
     PlaybackState_PlaybackSpeedMultiplier=~
     PlaybackState_bLevelStreaming=~
     PlaybackState_bHasRelevancyZone=~
     PlaybackState_bHasPlayInterval=~
     PlaybackState_PlayIntervalStartTime=~
     PlaybackState_PlayIntervalEndTime=~
     PlaybackState_bLoopInterval=~
     PlaybackState_IntervalLoopBufferTime=~
    )
   )
  ]}
 ShowMessage {nodes [
   n1 = ShowMessage!() -> (then, Notification:Text) {is=title K2Node_FunctionEntry}
   > n2 = Cast To FortReplaySpectator!(execute, Object=GetOwningPlayer()) -> (then, CastFailed, AsFort Replay Spectator:Fort Replay Spectator Object Reference) {is=title K2Node_DynamicCast}
   > ShowNotification!(execute, self:Fort Player Controller Spectating Object Reference=@n2, NotificationText:Text=@n1) -> (then)
  ]}
]
variables [
 bIsSizeSet {type=boolean}
 CanvasPanelSize {type=number}
 InTime {type=number}
 OutTime {type=number}
]
widgets {
 info {inheritedWidgetCount=6, parentClass=/Script/FortniteUI.FortTimelineBase, rootWidgetClass=&HorizontalBox, widgetCount=22}
 widgets [
  HorizontalBox_2 {
   widgetClassPath=&HorizontalBox
   widget=HorizontalBox_2
   properties {navigation=&self:WidgetTree.HorizontalBox_2.WidgetNavigation_0}
  }
  TimeElapsed {
   widgetClassPath=&Overlay
   parent=HorizontalBox_2
   widget=TimeElapsed
   properties {bOverride_Cursor=true}
   slot {class=HorizontalBoxSlot, properties {verticalAlignment=VAlign_Center}}
  }
  WBP_UIKit_Backplate {
   widgetClassPath=&WBP_UIKit_Backplate
   bIsVariable=true
   parent=TimeElapsed
   widget=WBP_UIKit_Backplate
   properties {backplateMaterial=/FortUILibrary/Foundation/Buttons/Materials/M_Timeline_Backplate_Inst, brightness=Medium, cornerRadius=Tab}
   slot {class=OverlaySlot, properties {horizontalAlignment=HAlign_Fill, verticalAlignment=VAlign_Fill}}
  }
  SizeBox_0 {
   widgetClassPath=&SizeBox
   parent=TimeElapsed
   widget=SizeBox_0
   properties {bOverride_HeightOverride=true, bOverride_WidthOverride=true, heightOverride=40, widthOverride=100}
   slot {class=OverlaySlot}
  }
  Text_CurrentTime {
   widgetClassPath=&CommonTextBlock
   bInherited=true
   parent=SizeBox_0
   widget=Text_CurrentTime
   properties {
    justification=Center
    text=00:00
    font {
     bForceMonospaced=false
     bMaterialIsStencil=false
     fontMaterial=null
     fontObject=&HeadingNow
     letterSpacing=0
     monospacedWidth=1
     size=15.75
     skewAmount=0
     typefaceFontName=74 Regular
     outlineSettings {
      bApplyOutlineToDropShadows=false
      bMiteredCorners=false
      bSeparateFillAlpha=false
      outlineMaterial=null
      outlineSize=0
      outlineColor=<rgba 0 0 0 1>
     }
    }
    shadowOffset=<0 0>
   }
   slot {class=SizeBoxSlot, properties {horizontalAlignment=HAlign_Center, verticalAlignment=VAlign_Center}}
  }
  SizeBoxTimelineBar {
   widgetClassPath=&SizeBox
   parent=HorizontalBox_2
   widget=SizeBoxTimelineBar
   properties {bOverride_HeightOverride=true, bOverride_WidthOverride=true, heightOverride=36, minDesiredWidth=288, widthOverride=800}
   slot {class=HorizontalBoxSlot, properties {verticalAlignment=VAlign_Center, size {sizeRule=Fill, value=1}}}
  }
  Border_2 {
   widgetClassPath=&Border
   parent=SizeBoxTimelineBar
   widget=Border_2
   properties {
    visibility=SelfHitTestInvisible
    background {
     bIsDynamicallyLoaded=false
     drawAs=Box
     imageType=NoImage
     mirroring=NoMirror
     resourceName=None
     resourceObject=/FortUILibrary/Styles/Border/Materials/M-Border-TimelineShadow
     tiling=NoTile
     imageSize=<32 32>
     margin=<ltrb 0.3 0.3 0.3 0.3>
     outlineSettings {
      bUseBrushTransparency=false
      roundingType=HalfHeightRadius
      width=0
      color {colorUseRule=UseColor_Specified, specifiedColor=<rgba 0 0 0 0>}
      cornerRadii=<0 0 0 0>
     }
     tintColor {colorUseRule=UseColor_Specified, specifiedColor=<rgba 1 1 1 1>}
     uVRegion {bIsValid=false, max=<0 0>, min=<0 0>}
    }
    brushColor=<rgba 1 1 1 0.5>
    padding=<ltrb 8 8 8 8>
   }
   slot {class=SizeBoxSlot, properties {padding=<ltrb 10 0 10 0>}}
  }
  Overlay_0 {widgetClassPath=&Overlay, parent=Border_2, widget=Overlay_0, slot {class=BorderSlot, properties {padding=<ltrb 8 8 8 8>}}}
  Border_0 {
   widgetClassPath=&Border
   parent=Overlay_0
   widget=Border_0
   properties {
    visibility=SelfHitTestInvisible
    background {
     bIsDynamicallyLoaded=false
     drawAs=Border
     imageType=NoImage
     mirroring=NoMirror
     resourceName=None
     resourceObject=/FortUILibrary/Styles/Border/Materials/M_Border_SeasonPassHeader_Active
     tiling=NoTile
     imageSize=<128 128>
     margin=<ltrb 0.03 0.03 0.03 0.03>
     outlineSettings {
      bUseBrushTransparency=false
      roundingType=HalfHeightRadius
      width=0
      color {colorUseRule=UseColor_Specified, specifiedColor=<rgba 0 0 0 0>}
      cornerRadii=<0 0 0 0>
     }
     tintColor {colorUseRule=UseColor_Specified, specifiedColor=<rgba 1 1 1 1>}
     uVRegion {bIsValid=false, max=<0 0>, min=<0 0>}
    }
   }
   slot {class=OverlaySlot, properties {horizontalAlignment=HAlign_Fill, verticalAlignment=VAlign_Fill}}
  }
  Image_75 {
   widgetClassPath=&Image
   bIsVariable=true
   parent=Overlay_0
   widget=Image_75
   properties {
    visibility=SelfHitTestInvisible
    brush {
     bIsDynamicallyLoaded=false
     drawAs=Box
     imageType=NoImage
     mirroring=NoMirror
     resourceName=None
     resourceObject=null
     tiling=NoTile
     imageSize=<32 32>
     margin=<ltrb 0 0 0 0>
     outlineSettings {
      bUseBrushTransparency=false
      roundingType=HalfHeightRadius
      width=0
      color {colorUseRule=UseColor_Specified, specifiedColor=<rgba 0 0 0 0>}
      cornerRadii=<0 0 0 0>
     }
     tintColor {colorUseRule=UseColor_Specified, specifiedColor=<rgba 1 1 1 1>}
     uVRegion {bIsValid=false, max=<0 0>, min=<0 0>}
    }
    colorAndOpacity=<rgba 0.1779 0.2542 0.3564 0.498>
   }
   slot {class=OverlaySlot, properties {horizontalAlignment=HAlign_Fill, verticalAlignment=VAlign_Fill}}
  }
  Overlay_ProgressBar {
   widgetClassPath=&Overlay
   bInherited=true
   parent=Overlay_0
   widget=Overlay_ProgressBar
   slot {class=OverlaySlot, properties {horizontalAlignment=HAlign_Fill, verticalAlignment=VAlign_Fill}}
  }
  ProgressBar_TimeElapsed {
   widgetClassPath=/Script/UMG.ProgressBar
   bInherited=true
   bIsVariable=true
   parent=Overlay_ProgressBar
   widget=ProgressBar_TimeElapsed
   properties {
    percent=0.33
    fillColorAndOpacity=<rgba 1 1 1 0.7>
    widgetStyle {
     enableFillAnimation=false
     backgroundImage {
      bIsDynamicallyLoaded=false
      drawAs=NoDrawType
      imageType=FullColor
      mirroring=NoMirror
      resourceName=../../../Engine/Content/Slate/Common/ProgressBar_Background.png
      resourceObject=null
      tiling=NoTile
      imageSize=<0 0>
      margin=<ltrb 0 0 0 0>
      outlineSettings {
       bUseBrushTransparency=false
       roundingType=HalfHeightRadius
       width=0
       color {colorUseRule=UseColor_Specified, specifiedColor=<rgba 0 0 0 0>}
       cornerRadii=<0 0 0 0>
      }
      tintColor {colorUseRule=UseColor_Specified, specifiedColor=<rgba 1 1 1 1>}
      uVRegion {bIsValid=false, max=<0 0>, min=<0 0>}
     }
     fillImage {
      bIsDynamicallyLoaded=false
      drawAs=Image
      imageType=FullColor
      mirroring=NoMirror
      resourceName=None
      resourceObject=null
      tiling=NoTile
      imageSize=<0 0>
      margin=<ltrb 0.4 0 0.4 0>
      outlineSettings {
       bUseBrushTransparency=false
       roundingType=HalfHeightRadius
       width=0
       color {colorUseRule=UseColor_Specified, specifiedColor=<rgba 0 0 0 0>}
       cornerRadii=<0 0 0 0>
      }
      tintColor {colorUseRule=UseColor_Specified, specifiedColor=<rgba 0.2623 0.3095 0.3613 0.8>}
      uVRegion {bIsValid=false, max=<0 0>, min=<0 0>}
     }
     marqueeImage {
      bIsDynamicallyLoaded=false
      drawAs=NoDrawType
      imageType=FullColor
      mirroring=NoMirror
      resourceName=../../../Engine/Content/Slate/Common/ProgressBar_Marquee.png
      resourceObject=null
      tiling=Both
      imageSize=<4 4>
      margin=<ltrb 0 0 0 0>
      outlineSettings {
       bUseBrushTransparency=false
       roundingType=HalfHeightRadius
       width=0
       color {colorUseRule=UseColor_Specified, specifiedColor=<rgba 0 0 0 0>}
       cornerRadii=<0 0 0 0>
      }
      tintColor {colorUseRule=UseColor_Specified, specifiedColor=<rgba 1 1 1 1>}
      uVRegion {bIsValid=false, max=<0 0>, min=<0 0>}
     }
    }
   }
   slot {class=OverlaySlot, properties {horizontalAlignment=HAlign_Fill, verticalAlignment=VAlign_Fill}}
  }
  Slider_Gamepad {
   widgetClassPath=/Script/CommonUI.AnalogSlider
   bInherited=true
   parent=Overlay_ProgressBar
   widget=Slider_Gamepad
   properties {
    navigation=&self:WidgetTree.Slider_Gamepad.WidgetNavigation_0
    stepSize=0.006
    value=0.33
    visibility=Collapsed
    sliderBarColor=<rgba 0.4115 0.4115 0.4115 1>
    sliderHandleColor=<rgba 1 0.4125 0 1>
    widgetStyle {
     barThickness=6
     disabledBarImage {
      bIsDynamicallyLoaded=false
      drawAs=Image
      imageType=NoImage
      mirroring=NoMirror
      resourceName=None
      resourceObject=null
      tiling=NoTile
      imageSize=<0 0>
      margin=<ltrb 0 0 0 0>
      outlineSettings {
       bUseBrushTransparency=false
       roundingType=HalfHeightRadius
       width=0
       color {colorUseRule=UseColor_Specified, specifiedColor=<rgba 0 0 0 0>}
       cornerRadii=<0 0 0 0>
      }
      tintColor {colorUseRule=UseColor_Specified, specifiedColor=<rgba 0.5 0.5 0.5 1>}
      uVRegion {bIsValid=false, max=<0 0>, min=<0 0>}
     }
     disabledThumbImage {
      bIsDynamicallyLoaded=false
      drawAs=Image
      imageType=FullColor
      mirroring=NoMirror
      resourceName=../../../Engine/Content/Slate/Common/Button_Disabled.png
      resourceObject=null
      tiling=NoTile
      imageSize=<8 14>
      margin=<ltrb 0 0 0 0>
      outlineSettings {
       bUseBrushTransparency=false
       roundingType=HalfHeightRadius
       width=0
       color {colorUseRule=UseColor_Specified, specifiedColor=<rgba 0 0 0 0>}
       cornerRadii=<0 0 0 0>
      }
      tintColor {colorUseRule=UseColor_Specified, specifiedColor=<rgba 1 1 1 1>}
      uVRegion {bIsValid=false, max=<0 0>, min=<0 0>}
     }
     hoveredBarImage {
      bIsDynamicallyLoaded=false
      drawAs=Image
      imageType=NoImage
      mirroring=NoMirror
      resourceName=None
      resourceObject=null
      tiling=NoTile
      imageSize=<0 0>
      margin=<ltrb 0 0 0 0>
      outlineSettings {
       bUseBrushTransparency=false
       roundingType=HalfHeightRadius
       width=0
       color {colorUseRule=UseColor_Specified, specifiedColor=<rgba 0 0 0 0>}
       cornerRadii=<0 0 0 0>
      }
      tintColor {colorUseRule=UseColor_Specified, specifiedColor=<rgba 1 1 1 1>}
      uVRegion {bIsValid=false, max=<0 0>, min=<0 0>}
     }
     hoveredThumbImage {
      bIsDynamicallyLoaded=false
      drawAs=Image
      imageType=FullColor
      mirroring=NoMirror
      resourceName=../../../Engine/Content/Slate/Common/Button.png
      resourceObject=null
      tiling=NoTile
      imageSize=<8 14>
      margin=<ltrb 0 0 0 0>
      outlineSettings {
       bUseBrushTransparency=false
       roundingType=HalfHeightRadius
       width=0
       color {colorUseRule=UseColor_Specified, specifiedColor=<rgba 0 0 0 0>}
       cornerRadii=<0 0 0 0>
      }
      tintColor {colorUseRule=UseColor_Specified, specifiedColor=<rgba 1 1 1 1>}
      uVRegion {bIsValid=false, max=<0 0>, min=<0 0>}
     }
     normalBarImage {
      bIsDynamicallyLoaded=false
      drawAs=Image
      imageType=NoImage
      mirroring=NoMirror
      resourceName=None
      resourceObject=null
      tiling=NoTile
      imageSize=<0 0>
      margin=<ltrb 0 0 0 0>
      outlineSettings {
       bUseBrushTransparency=false
       roundingType=HalfHeightRadius
       width=0
       color {colorUseRule=UseColor_Specified, specifiedColor=<rgba 0 0 0 0>}
       cornerRadii=<0 0 0 0>
      }
      tintColor {colorUseRule=UseColor_Specified, specifiedColor=<rgba 1 1 1 1>}
      uVRegion {bIsValid=false, max=<0 0>, min=<0 0>}
     }
     normalThumbImage {
      bIsDynamicallyLoaded=false
      drawAs=Image
      imageType=FullColor
      mirroring=NoMirror
      resourceName=../../../Engine/Content/Slate/Common/Button.png
      resourceObject=/Game/UI/Foundation/Textures/Icons/Replay/T-Icon-Replay-RoundSquare-L
      tiling=NoTile
      imageSize=<16 32>
      margin=<ltrb 0 0 0 0>
      outlineSettings {
       bUseBrushTransparency=false
       roundingType=HalfHeightRadius
       width=0
       color {colorUseRule=UseColor_Specified, specifiedColor=<rgba 0 0 0 0>}
       cornerRadii=<0 0 0 0>
      }
      tintColor {colorUseRule=UseColor_Specified, specifiedColor=<rgba 1 1 1 1>}
      uVRegion {bIsValid=false, max=<0 0>, min=<0 0>}
     }
    }
   }
   slot {class=OverlaySlot, properties {horizontalAlignment=HAlign_Fill, verticalAlignment=VAlign_Fill}}
  }
  CanvasPanel_Markers {
   widgetClassPath=&CanvasPanel
   bInherited=true
   parent=Overlay_ProgressBar
   widget=CanvasPanel_Markers
   slot {class=OverlaySlot, properties {horizontalAlignment=HAlign_Fill, verticalAlignment=VAlign_Fill}}
  }
  CanvasPanelIndicators {
   widgetClassPath=&CanvasPanel
   bIsVariable=true
   parent=Overlay_ProgressBar
   widget=CanvasPanelIndicators
   slot {class=OverlaySlot, properties {horizontalAlignment=HAlign_Fill, verticalAlignment=VAlign_Fill}}
  }
  In_Indicator {
   widgetClassPath=&Image
   bIsVariable=true
   parent=CanvasPanelIndicators
   widget=In_Indicator
   properties {visibility=Collapsed}
   slot {class=CanvasPanelSlot, properties {layoutData {alignment=<0 0>, anchors {maximum=<0 0>, minimum=<0 0>}, offsets=<ltrb 0 0 5 20>}}}
  }
  Out_Indicator {
   widgetClassPath=&Image
   bIsVariable=true
   parent=CanvasPanelIndicators
   widget=Out_Indicator
   properties {visibility=Collapsed}
   slot {class=CanvasPanelSlot, properties {layoutData {alignment=<0 0>, anchors {maximum=<0 0>, minimum=<0 0>}, offsets=<ltrb 0 0 5 20>}}}
  }
  TimeEnding {
   widgetClassPath=&Overlay
   parent=HorizontalBox_2
   widget=TimeEnding
   properties {bOverride_Cursor=true}
   slot {class=HorizontalBoxSlot, properties {verticalAlignment=VAlign_Center}}
  }
  WBP_UIKit_Backplate_1 {
   widgetClassPath=&WBP_UIKit_Backplate
   bIsVariable=true
   parent=TimeEnding
   widget=WBP_UIKit_Backplate_1
   properties {backplateMaterial=&M_Timeline_Backplate_Inst, brightness=Medium, cornerRadius=Tab}
   slot {class=OverlaySlot, properties {horizontalAlignment=HAlign_Fill, verticalAlignment=VAlign_Fill}}
  }
  SizeBox {
   widgetClassPath=&SizeBox
   parent=TimeEnding
   widget=SizeBox
   properties {bOverride_HeightOverride=true, bOverride_WidthOverride=true, heightOverride=40, widthOverride=100}
   slot {class=OverlaySlot}
  }
  Text_EndTime {
   widgetClassPath=&CommonTextBlock
   bInherited=true
   parent=SizeBox
   widget=Text_EndTime
   properties {
    justification=Center
    text=00:00
    font {
     bForceMonospaced=false
     bMaterialIsStencil=false
     fontMaterial=null
     fontObject=&HeadingNow
     letterSpacing=0
     monospacedWidth=1
     size=15.75
     skewAmount=0
     typefaceFontName=74 Regular
     outlineSettings {
      bApplyOutlineToDropShadows=false
      bMiteredCorners=false
      bSeparateFillAlpha=false
      outlineMaterial=null
      outlineSize=0
      outlineColor=<rgba 0 0 0 1>
     }
    }
    shadowOffset=<0 0>
   }
   slot {class=SizeBoxSlot, properties {horizontalAlignment=HAlign_Center, verticalAlignment=VAlign_Center, padding=<ltrb 4 0 4 0>}}
  }
  SizeBox_105 {
   widgetClassPath=&SizeBox
   parent=HorizontalBox_2
   widget=SizeBox_105
   properties {bOverride_HeightOverride=true, heightOverride=60}
   slot {class=HorizontalBoxSlot}
  }
 ]
}
